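# Control Rig for the female Photographer body's jacket, correcting jacket, tied-hoodie and camera bones, mainly when rigid-body simulation is off.
# A shared deform-rig function runs first, then Hoodie, Jacket, Camera and Additional Fixed functions.
# State booleans, some smoothed, weight bone offsets and parent constraints to spine, pelvis and thighs.
# Spherical pose readers on thighs and spine_02, scaled by GoldenGloryEmote, drive extra hoodie and jacket corrections.
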ControlRigBlueprint /BRCosmetics/Characters/Player/Female/Medium/Bodies/F_MED_Photographer/Meshes/F_MED_Photographer_Jacket_CtrlRig
graphs [
 Multiply_2_ContainedGraph {nodes [
   n1 = RigVMFunctionEntryNode()
   RigVMFunctionReturnNode(Result=&RigVMFunction_MathDoubleMul(B=@n1.C, A=&RigVMFunction_MathDoubleMul(A=@n1.A, B=@n1.B).Result).Result)
  ]}
 New Function_ContainedGraph {nodes [
   RigVMFunctionEntryNode()
   RigVMFunction_MathBoolNot = &RigVMFunction_MathBoolNot(Value=RigVMVariableNode() {variable=isCrouch}.Value)
   RigVMFunction_MathBoolToFloat = &RigVMFunction_MathBoolToFloat(Value=&RigVMFunction_MathBoolAnd(B=@RigVMFunction_MathBoolNot.Result, A=RigVMVariableNode() {variable=noRBAN_Mov}.Value).Result)
   n1 = &RigVMFunction_AlphaInterp(InterpSpeedDecreasing=10, InterpSpeedIncreasing=10, Scale=1, bClampResult=false, bMapRange=true, Value=@RigVMFunction_MathBoolToFloat.Result)
   > &RigUnit_ModifyTransforms(
    ItemToModify="((Item=(Type=Bone,Name=\"dyn_hoodie_01_l\"),Transform=(Rotation=(X=0,Y=0,Z=-0.0872,W=0.9962),Translation=(X=0,Y=-2,Z=0),Scale3D=(X=1,Y=1,Z=1))),(Item=(Type=Bone,Name=\"dyn_hoodie_02_l\"),Transform=(Rotation=(X=0,Y=0,Z=-0.0436,W=0.999),Translation=(X=0,Y=0,Z=0) ...(truncated)"
    WeightMinimum=0
    Weight=@n1.Result
   )
   > &RigUnit_ModifyTransforms(
    ItemToModify="((Item=(Type=Bone,Name=\"dyn_hoodie_01_c\"),Transform=(Rotation=(X=0,Y=0,Z=-0.0872,W=0.9962),Translation=(X=0,Y=-2,Z=0),Scale3D=(X=1,Y=1,Z=1))),(Item=(Type=Bone,Name=\"dyn_hoodie_02_c\"),Transform=(Rotation=(X=0,Y=0,Z=-0.0436,W=0.999),Translation=(X=0,Y=0,Z=0) ...(truncated)"
    WeightMinimum=0
    Weight=@n1.Result
   )
   > &RigUnit_ModifyTransforms(
    ItemToModify="((Item=(Type=Bone,Name=\"dyn_hoodie_01_r\"),Transform=(Rotation=(X=0,Y=0,Z=-0.0872,W=0.9962),Translation=(X=0,Y=-2,Z=0),Scale3D=(X=1,Y=1,Z=1))),(Item=(Type=Bone,Name=\"dyn_hoodie_02_r\"),Transform=(Rotation=(X=0,Y=0,Z=-0.0436,W=0.999),Translation=(X=0,Y=0,Z=0) ...(truncated)"
    WeightMinimum=0
    Weight=@n1.Result
   )
   RigVMFunction_MathBoolAnd = &RigVMFunction_MathBoolAnd(A=RigVMVariableNode() {variable=noRigidBody}.Value, B=RigVMVariableNode() {variable=isSkydiving}.Value)
   n2 = &RigVMFunction_AlphaInterp(
    InterpSpeedDecreasing=10
    InterpSpeedIncreasing=10
    Scale=1
    bClampResult=false
    bMapRange=true
    Value=&RigVMFunction_MathBoolToFloat(Value=@RigVMFunction_MathBoolAnd.Result).Result
   )
   > &RigUnit_ModifyTransforms(
    ItemToModify="((Item=(Type=Bone,Name=\"dyn_hoodie_01_l\"),Transform=(Rotation=(X=0,Y=0,Z=-0.1305,W=0.9914),Translation=(X=0,Y=-2,Z=0),Scale3D=(X=1,Y=1,Z=1))),(Item=(Type=Bone,Name=\"dyn_hoodie_02_l\"),Transform=(Rotation=(X=0,Y=0,Z=-0.0436,W=0.999),Translation=(X=0,Y=0,Z=0) ...(truncated)"
    WeightMinimum=0
    Weight=@n2.Result
   )
   > &RigUnit_ModifyTransforms(
    ItemToModify="((Item=(Type=Bone,Name=\"dyn_hoodie_01_c\"),Transform=(Rotation=(X=0,Y=0,Z=-0.1305,W=0.9914),Translation=(X=0,Y=-2,Z=0),Scale3D=(X=1,Y=1,Z=1))),(Item=(Type=Bone,Name=\"dyn_hoodie_02_c\"),Transform=(Rotation=(X=0,Y=0,Z=-0.0436,W=0.999),Translation=(X=0,Y=0,Z=0) ...(truncated)"
    WeightMinimum=0
    Weight=@n2.Result
   )
   > &RigUnit_ModifyTransforms(
    ItemToModify="((Item=(Type=Bone,Name=\"dyn_hoodie_01_r\"),Transform=(Rotation=(X=0,Y=0,Z=-0.1305,W=0.9914),Translation=(X=0,Y=-2,Z=0),Scale3D=(X=1,Y=1,Z=1))),(Item=(Type=Bone,Name=\"dyn_hoodie_02_r\"),Transform=(Rotation=(X=0,Y=0,Z=-0.0436,W=0.999),Translation=(X=0,Y=0,Z=0) ...(truncated)"
    WeightMinimum=0
    Weight=@n2.Result
   )
   n3 = RigVMVariableNode() {variable=noRigidBody}
   n4 = &RigVMFunction_MathBoolToFloat(Value=@n3.Value)
   > &RigUnit_ParentConstraint(
    AdvancedSettings=(InterpolationType=Average,RotationOrderForFilter=XZY)
    Child=(Type=Bone,Name="dyn_hoodieFront_01_l")
    Parents=((Item=(Type=Bone,Name="thigh_l"),Weight=0.5),(Item=(Type=Bone,Name="pelvis"),Weight=0.5))
    Weight=@n4.Result
   )
   > &RigUnit_ParentConstraint(
    AdvancedSettings=(InterpolationType=Average,RotationOrderForFilter=XZY)
    Child=(Type=Bone,Name="dyn_hoodieFront_01_r")
    Parents=((Item=(Type=Bone,Name="thigh_r"),Weight=0.5),(Item=(Type=Bone,Name="pelvis"),Weight=0.5))
    Weight=@n4.Result
   )
   n5 = &RigVMFunction_MathBoolNot(Value=RigVMVariableNode() {variable=isFrontEnd}.Value)
   n6 = &RigVMFunction_AlphaInterp(
    InterpSpeedDecreasing=10
    InterpSpeedIncreasing=10
    Scale=1
    bClampResult=false
    bMapRange=true
    Value=&RigVMFunction_MathBoolToFloat(Value=&RigVMFunction_MathBoolAnd(B=@n5.Result, A=@n3.Value).Result).Result
   )
   > &RigUnit_ModifyTransforms(
    ItemToModify=((Item=(Type=Bone,Name="dyn_hoodieKnot"),Transform=(Rotation=(X=0,Y=0,Z=-0.0872,W=0.9962),Translation=(X=0,Y=0,Z=0),Scale3D=(X=1,Y=1,Z=1))))
    WeightMinimum=0
    Weight=@n6.Result
   )
   > n7 = &RigUnit_SphericalPoseReader(
    ActiveRegionScaleFactors=(PositiveWidth=1,NegativeWidth=1,PositiveHeight=1,NegativeHeight=1)
    ActiveRegionSize=0.05
    Debug=(bDrawDebug=false,bDraw2D=False,bDrawLocalAxes=False,DebugScale=25,DebugSegments=20,DebugThickness=0.25)
    DriverAxis=(X=-1,Y=0,Z=0)
    DriverItem=(Type=Bone,Name="thigh_l")
    FalloffRegionScaleFactors=(PositiveWidth=1,NegativeWidth=1,PositiveHeight=1,NegativeHeight=1)
    FalloffSize=0.55
    FlipHeightScaling=False
    FlipWidthScaling=False
    OptionalParentItem=(Type=Bone,Name="None")
    RotationOffset=(X=-90,Y=0,Z=-10)
   )
   AlphaInterp = &RigVMFunction_AlphaInterp(InterpSpeedDecreasing=10, InterpSpeedIncreasing=10, Scale=1, bClampResult=true, Value=RigVMVariableNode() {variable=GoldenGloryEmote}.Value)
   Multiply = &RigVMFunction_MathDoubleMul(B=@AlphaInterp.Result, A=@n7.OutputParam)
   > &RigUnit_ModifyTransforms(
    ItemToModify="((Item=(Type=Bone,Name=\"dyn_hoodie_01_Base_l\"),Transform=(Rotation=(X=0,Y=0,Z=0.4338,W=-0.901),Translation=(X=-9.5226,Y=-7.5969,Z=-0.0001),Scale3D=(X=1,Y=1,Z=1))),(Item=(Type=Bone,Name=\"dyn_hoodie_01_c_Base\"),Transform=(Rotation=(X=0,Y=0,Z=-0.2636,W=0.9646 ...(truncated)"
    WeightMinimum=0
    Weight=@Multiply.Result
   )
   > n8 = &RigUnit_SphericalPoseReader(
    ActiveRegionScaleFactors=(PositiveWidth=1,NegativeWidth=1,PositiveHeight=1,NegativeHeight=1)
    ActiveRegionSize=0.15
    Debug=(bDrawDebug=true,bDraw2D=False,bDrawLocalAxes=False,DebugScale=25,DebugSegments=20,DebugThickness=0.25)
    DriverAxis=(X=1,Y=0,Z=0)
    DriverItem=(Type=Bone,Name="spine_02")
    FalloffRegionScaleFactors=(PositiveWidth=1,NegativeWidth=1,PositiveHeight=1,NegativeHeight=1)
    FalloffSize=0.4
    FlipHeightScaling=False
    FlipWidthScaling=False
    OptionalParentItem=(Type=Bone,Name="root")
    RotationOffset=(X=90,Y=0,Z=0)
   )
   > &RigUnit_ModifyTransforms(
    ItemToModify="((Item=(Type=Bone,Name=\"dyn_hoodieFront_01_Base_l\"),Transform=(Rotation=(X=0,Y=0,Z=0.6549,W=-0.7557),Translation=(X=0,Y=0,Z=0),Scale3D=(X=1,Y=1,Z=1))),(Item=(Type=Bone,Name=\"dyn_hoodieFront_01_Base_r\"),Transform=(Rotation=(X=0,Y=0,Z=0.0869,W=0.9962),Transl ...(truncated)"
    WeightMinimum=0
    Weight=&RigVMFunction_MathDoubleMul(B=@AlphaInterp.Result, A=@n8.OutputParam).Result
   )
   > RigVMFunctionReturnNode()
  ]}
 New Function_ContainedGraph {nodes [
   RigVMFunctionEntryNode()
   RigVMFunction_MathBoolNot = &RigVMFunction_MathBoolNot(Value=RigVMVariableNode() {variable=isFrontEnd}.Value)
   n1 = RigVMVariableNode() {variable=noRigidBody}
   RigVMFunction_MathBoolAnd = &RigVMFunction_MathBoolAnd(B=@RigVMFunction_MathBoolNot.Result, A=@n1.Value)
   > &RigUnit_ModifyTransforms(
    ItemToModify="((Item=(Type=Bone,Name=\"dyn_jacket_fr_l_02\"),Transform=(Rotation=(X=0,Y=0,Z=-0.1736,W=0.9848),Translation=(X=0,Y=1,Z=0),Scale3D=(X=1,Y=1,Z=1))),(Item=(Type=Bone,Name=\"dyn_jacket_fr_r_02\"),Transform=(Rotation=(X=0,Y=0,Z=-0.0872,W=0.9962),Translation=(X=0,Y= ...(truncated)"
    WeightMinimum=0
    Weight=&RigVMFunction_MathBoolToFloat(Value=@RigVMFunction_MathBoolAnd.Result).Result
   )
   n2 = &RigVMFunction_MathBoolToFloat(Value=@n1.Value)
   > &RigUnit_ParentConstraint(
    AdvancedSettings=(InterpolationType=Average,RotationOrderForFilter=XZY)
    Child=(Type=Bone,Name="dyn_jacket_fr_l_01")
    Parents=((Item=(Type=Bone,Name="spine_03"),Weight=0.5))
    Weight=@n2.Result
   )
   > &RigUnit_ParentConstraint(
    AdvancedSettings=(InterpolationType=Average,RotationOrderForFilter=XZY)
    Child=(Type=Bone,Name="dyn_jacket_l_01")
    Parents=((Item=(Type=Bone,Name="spine_03"),Weight=0.5))
    Weight=@n2.Result
   )
   > &RigUnit_ParentConstraint(
    AdvancedSettings=(InterpolationType=Average,RotationOrderForFilter=XZY)
    Child=(Type=Bone,Name="dyn_jacket_bk_l_01")
    Parents=((Item=(Type=Bone,Name="spine_03"),Weight=0.5))
    Weight=@n2.Result
   )
   > &RigUnit_ParentConstraint(
    AdvancedSettings=(InterpolationType=Average,RotationOrderForFilter=XZY)
    Child=(Type=Bone,Name="dyn_jacket_bk_01")
    Parents=((Item=(Type=Bone,Name="spine_03"),Weight=0.5))
    Weight=@n2.Result
   )
   > &RigUnit_ParentConstraint(
    AdvancedSettings=(InterpolationType=Average,RotationOrderForFilter=XZY)
    Child=(Type=Bone,Name="dyn_jacket_bk_r_01_Base")
    Parents=((Item=(Type=Bone,Name="spine_03"),Weight=0.5))
    Weight=@n2.Result
   )
   > &RigUnit_ParentConstraint(
    AdvancedSettings=(InterpolationType=Average,RotationOrderForFilter=XZY)
    Child=(Type=Bone,Name="dyn_jacket_r_01")
    Parents=((Item=(Type=Bone,Name="spine_03"),Weight=0.5))
    Weight=@n2.Result
   )
   > &RigUnit_ParentConstraint(
    AdvancedSettings=(InterpolationType=Average,RotationOrderForFilter=XZY)
    Child=(Type=Bone,Name="dyn_jacket_fr_r_01_Base")
    Parents=((Item=(Type=Bone,Name="spine_03"),Weight=0.5))
    Weight=@n2.Result
   )
   > &RigUnit_ParentConstraint(
    AdvancedSettings=(InterpolationType=Average,RotationOrderForFilter=XZY)
    Child=(Type=Bone,Name="dyn_jacket_fr_l_02")
    Parents=((Item=(Type=Bone,Name="spine_02"),Weight=1))
    Weight=@n2.Result
   )
   > &RigUnit_ParentConstraint(
    AdvancedSettings=(InterpolationType=Average,RotationOrderForFilter=XZY)
    Child=(Type=Bone,Name="dyn_jacket_l_02")
    Parents=((Item=(Type=Bone,Name="spine_02"),Weight=1))
    Weight=@n2.Result
   )
   > &RigUnit_ParentConstraint(
    AdvancedSettings=(InterpolationType=Average,RotationOrderForFilter=XZY)
    Child=(Type=Bone,Name="dyn_jacket_bk_l_02")
    Parents=((Item=(Type=Bone,Name="spine_02"),Weight=1))
    Weight=@n2.Result
   )
   > &RigUnit_ParentConstraint(
    AdvancedSettings=(InterpolationType=Average,RotationOrderForFilter=XZY)
    Child=(Type=Bone,Name="dyn_jacket_bk_02")
    Parents=((Item=(Type=Bone,Name="spine_02"),Weight=1))
    Weight=@n2.Result
   )
   > &RigUnit_ParentConstraint(
    AdvancedSettings=(InterpolationType=Average,RotationOrderForFilter=XZY)
    Child=(Type=Bone,Name="dyn_jacket_bk_r_02")
    Parents=((Item=(Type=Bone,Name="spine_02"),Weight=1))
    Weight=@n2.Result
   )
   > &RigUnit_ParentConstraint(
    AdvancedSettings=(InterpolationType=Average,RotationOrderForFilter=XZY)
    Child=(Type=Bone,Name="dyn_jacket_r_02")
    Parents=((Item=(Type=Bone,Name="spine_02"),Weight=1))
    Weight=@n2.Result
   )
   > &RigUnit_ParentConstraint(
    AdvancedSettings=(InterpolationType=Average,RotationOrderForFilter=XZY)
    Child=(Type=Bone,Name="dyn_jacket_fr_r_01")
    Parents=((Item=(Type=Bone,Name="spine_02"),Weight=1))
    Weight=@n2.Result
   )
   > &RigUnit_ParentConstraint(
    AdvancedSettings=(InterpolationType=Average,RotationOrderForFilter=XZY)
    Child=(Type=Bone,Name="dyn_jacket_fr_l_03")
    Parents=((Item=(Type=Bone,Name="spine_01"),Weight=1),(Item=(Type=Bone,Name="pelvis"),Weight=1))
    Weight=1
    Parents.0.Weight=@n2.Result
   )
   > &RigUnit_ParentConstraint(
    AdvancedSettings=(InterpolationType=Average,RotationOrderForFilter=XZY)
    Child=(Type=Bone,Name="dyn_jacket_l_03")
    Parents=((Item=(Type=Bone,Name="spine_01"),Weight=1),(Item=(Type=Bone,Name="pelvis"),Weight=1))
    Weight=@n2.Result
   )
   > &RigUnit_ParentConstraint(
    AdvancedSettings=(InterpolationType=Average,RotationOrderForFilter=XZY)
    Child=(Type=Bone,Name="dyn_jacket_bk_l_03")
    Parents=((Item=(Type=Bone,Name="spine_01"),Weight=1),(Item=(Type=Bone,Name="pelvis"),Weight=1))
    Weight=@n2.Result
   )
   > &RigUnit_ParentConstraint(
    AdvancedSettings=(InterpolationType=Average,RotationOrderForFilter=XZY)
    Child=(Type=Bone,Name="dyn_jacket_bk_03")
    Parents=((Item=(Type=Bone,Name="spine_01"),Weight=1),(Item=(Type=Bone,Name="pelvis"),Weight=1))
    Weight=@n2.Result
   )
   > &RigUnit_ParentConstraint(
    AdvancedSettings=(InterpolationType=Average,RotationOrderForFilter=XZY)
    Child=(Type=Bone,Name="dyn_jacket_bk_r_03")
    Parents=((Item=(Type=Bone,Name="spine_01"),Weight=1),(Item=(Type=Bone,Name="pelvis"),Weight=1))
    Weight=@n2.Result
   )
   > &RigUnit_ParentConstraint(
    AdvancedSettings=(InterpolationType=Average,RotationOrderForFilter=XZY)
    Child=(Type=Bone,Name="dyn_jacket_r_03")
    Parents=((Item=(Type=Bone,Name="spine_01"),Weight=1),(Item=(Type=Bone,Name="pelvis"),Weight=1))
    Weight=@n2.Result
   )
   > &RigUnit_ParentConstraint(
    AdvancedSettings=(InterpolationType=Average,RotationOrderForFilter=XZY)
    Child=(Type=Bone,Name="dyn_jacket_fr_r_03")
    Parents=((Item=(Type=Bone,Name="spine_01"),Weight=1),(Item=(Type=Bone,Name="pelvis"),Weight=1))
    Weight=@n2.Result
   )
   > RigVMFunctionReturnNode()
  ]}
 New Function_ContainedGraph {nodes [
   RigVMFunctionEntryNode()
   RigVMFunction_MathBoolOr = &RigVMFunction_MathBoolOr(B=RigVMVariableNode() {variable=noRigidBody}.Value, A=RigVMVariableNode() {variable=isFrontEnd}.Value)
   RigVMFunction_MathBoolToFloat = &RigVMFunction_MathBoolToFloat(Value=@RigVMFunction_MathBoolOr.Result)
   > &RigUnit_ModifyTransforms(
    ItemToModify=((Item=(Type=Bone,Name="dyn_camMotion_01"),Transform=(Rotation=(X=0,Y=0,Z=0.0087,W=1),Translation=(X=0,Y=1.2418,Z=0),Scale3D=(X=1,Y=1,Z=1))))
    WeightMinimum=0
    Weight=@RigVMFunction_MathBoolToFloat.Result
   )
   > RigVMFunctionReturnNode()
  ]}
 New Function_ContainedGraph {nodes [
   RigVMFunctionEntryNode()
   > n1 = &RigUnit_SphericalPoseReader(
    ActiveRegionScaleFactors=(PositiveWidth=1,NegativeWidth=1,PositiveHeight=1,NegativeHeight=1)
    ActiveRegionSize=0.05
    Debug=(bDrawDebug=false,bDraw2D=False,bDrawLocalAxes=False,DebugScale=25,DebugSegments=20,DebugThickness=0.25)
    DriverAxis=(X=1,Y=0,Z=0)
    DriverItem=(Type=Bone,Name="thigh_r")
    FalloffRegionScaleFactors=(PositiveWidth=1,NegativeWidth=1.1,PositiveHeight=1,NegativeHeight=1)
    FalloffSize=0.55
    FlipHeightScaling=true
    FlipWidthScaling=true
    OptionalParentItem=(Type=Bone,Name="None")
    RotationOffset=(X=-90,Y=0,Z=-20)
   )
   Remap = &RigVMFunction_MathDoubleRemap(SourceMaximum=0.6, SourceMinimum=0, TargetMaximum=1, TargetMinimum=0, Value=@n1.OutputParam)
   Multiply = &RigVMFunction_MathDoubleMul(A=@Remap.Result, B=RigVMVariableNode() {variable=GoldenGloryEmote}.Value)
   > &RigUnit_ModifyTransforms(
    ItemToModify=((Item=(Type=Bone,Name="dyn_hoodieFront_01_Base_r"),Transform=(Rotation=(X=-0.0045,Y=0.2956,Z=-0.6999,W=0.6502),Translation=(X=0,Y=0,Z=0),Scale3D=(X=1,Y=1,Z=1))))
    WeightMinimum=0
    Weight=@Multiply.Result
   )
   > n2 = &RigUnit_SphericalPoseReader(
    ActiveRegionScaleFactors=(PositiveWidth=1,NegativeWidth=1,PositiveHeight=1,NegativeHeight=1)
    ActiveRegionSize=0.15
    Debug=(bDrawDebug=false,bDraw2D=False,bDrawLocalAxes=False,DebugScale=25,DebugSegments=20,DebugThickness=0.25)
    DriverAxis=(X=1,Y=0,Z=0)
    DriverItem=(Type=Bone,Name="spine_02")
    FalloffRegionScaleFactors=(PositiveWidth=1,NegativeWidth=1,PositiveHeight=1,NegativeHeight=1)
    FalloffSize=0.4
    FlipHeightScaling=False
    FlipWidthScaling=False
    OptionalParentItem=(Type=Bone,Name="root")
    RotationOffset=(X=90,Y=0,Z=0)
   )
   RigVMFunction_MathBoolToFloat = &RigVMFunction_MathBoolToFloat(Value=RigVMVariableNode() {variable=RigidBody}.Value)
   AlphaInterp = &RigVMFunction_AlphaInterp(InterpSpeedDecreasing=10, InterpSpeedIncreasing=10, Scale=1, bClampResult=true, Value=@RigVMFunction_MathBoolToFloat.Result)
   Multiply_1 = RigVMAggregateNode(C=@AlphaInterp.Result, A=@n2.OutputParam, B=RigVMVariableNode() {variable=GoldenGloryEmote}.Value)
   > &RigUnit_ModifyTransforms(
    ItemToModify="((Item=(Type=Bone,Name=\"dyn_jacket_bk_l_01_Base\"),Transform=(Rotation=(X=0.0295,Y=-0.113,Z=-0.2414,W=0.9634),Translation=(X=0,Y=0,Z=0),Scale3D=(X=1,Y=1,Z=1))),(Item=(Type=Bone,Name=\"dyn_jacket_l_01_Base\"),Transform=(Rotation=(X=0,Y=0,Z=-0.4905,W=0.8715),Tr ...(truncated)"
    WeightMinimum=0
    Weight=@Multiply_1.Result
   )
   > RigVMFunctionReturnNode()
  ]}
 RigVMFunctionLibrary {nodes [
   Additional Fixed = RigVMCollapseNode(ExecuteContext=())
   Camera FrontEnd Setup = RigVMCollapseNode(ExecuteContext=())
   Hoodie No RBAN Setup = RigVMCollapseNode(ExecuteContext=())
   Jacket No RBAN Setup = RigVMCollapseNode(ExecuteContext=())
  ]}
 RigVMModel {nodes [
   n1 = &RigUnit_BeginExecution()
   FN_DeformRig_FNC = RigVMFunctionReferenceNode(
    DeformRig_ Curve="(deformRig_l_upperarm_up_alpha_72_#1=False,deformRig_r_upperarm_up_alpha_73_#2=False,deformRig_l_upperarm_fr_alpha_74_#3=False,deformRig_r_upperarm_fr_alpha_75_CB51B8 ...(truncated)"
    DeformRig_Bool="(upperarm_1_#4=True,elbow_3_#5=True,wrist_7_#6=True,groin_8_#7=True,glute_17_#8=True,knee_9_40C408324CA261 ...(truncated)"
    DeformRig_Debug=(upperarm_debug_18_#4=False,wrist_debug_20_#6=False,lat_debug_26_#9=False,groinGlute_debug_27_#7=False)
    DeformRig_Value="(upperarm_up_18_#10=(Rotation=(X=0.0,Y=-0.0,Z=0.0,W=1.0),Translation=(X=0.0,Y=0.0,Z=8.0),Scale3D=(X=0.70,Y=1.0,Z=1.0)),upperarm_fwd_fr_19_#11=(Rotation=(X=0.0,Y=-0.0,Z=0.0,W=1.0),Translation=(X=0.0, ...(truncated)"
    ExecuteContext=@n1.ExecutePin
   )
   > Hoodie No RBAN Setup = RigVMFunctionReferenceNode()
   > Jacket No RBAN Setup = RigVMFunctionReferenceNode()
   > Camera FrontEnd Setup = RigVMFunctionReferenceNode()
   > Additional Fixed = RigVMFunctionReferenceNode()
  ]}
]
hierarchy [
 root {loc=<0 0 0>, rot=<0 0 0 1>}
 pelvis {parent=root, loc=<0 -0.5301 98.0571>, rot=<0.0127 0.707 -0.0127 -0.707>}
 spine_01 {parent=pelvis, loc=<6.1208 -0.9328 0>, rot=<0 0 0.034 -0.9994>}
 spine_02 {parent=spine_01, loc=<6.5996 -0.2095 0>, rot=<0 0 0.0238 -0.9997>}
 spine_03 {parent=spine_02, loc=<6.4914 -0.1237 0>, rot=<0 0 0.0149 -0.9999>}
 spine_04 {parent=spine_03, loc=<6.9598 0.0794 0>, rot=<0 0 -0.0132 -0.9999>}
 spine_05 {parent=spine_04, loc=<6.9987 0 0>, rot=<0 0 -0.0015 -1>}
 clavicle_l {parent=spine_05, loc=<5.4831 -0.1053 -3.2946>, rot=<0.0465 -0.7289 -0.009 -0.6829>}
 upperarm_l {parent=clavicle_l, loc=<9.8658 -1.2136 -0.4105>, rot=<0.0164 -0.3245 -0.0269 -0.9454>}
 lowerarm_l {parent=upperarm_l, loc=<26.465 -0.8345 0.259>, rot=<0.0036 0.0161 -0.2298 -0.9731>}
 hand_l {parent=lowerarm_l, loc=<22.7355 -0.3203 -0.6105>, rot=<-0.6784 -0.0223 0.0735 -0.7307>}
 index_metacarpal_l {parent=hand_l, loc=<1.6535 0.0183 -2.96>, rot=<0.0158 -0.0124 -0.0173 -0.9996>}
 index_01_l {parent=index_metacarpal_l, loc=<8.3008 0.0357 -0.0524>, rot=<-0.0044 0.0485 0.0924 -0.9945>}
 index_02_l {parent=index_01_l, loc=<3.3587 -0.3003 0.0452>, rot=<-0.002 0.0027 0.3225 -0.9466>}
 index_03_l {parent=index_02_l, loc=<2.5578 -0.123 -0.0252>, rot=<-0.002 0.0067 0.2558 -0.9667>}
 middle_metacarpal_l {parent=hand_l, loc=<2.1856 -0.1725 -0.5984>, rot=<0.006 0.0168 -0.0329 -0.9993>}
 middle_01_l {parent=middle_metacarpal_l, loc=<7.3275 -0.2447 0.0129>, rot=<-0.0208 0.0042 0.0858 -0.9961>}
 middle_02_l {parent=middle_01_l, loc=<4.0123 0.1401 0.1175>, rot=<-0.0218 0.0258 0.3512 -0.9357>}
 middle_03_l {parent=middle_02_l, loc=<2.6077 0.0534 0.1373>, rot=<0.0082 0.0188 0.1708 -0.9851>}
 pinky_metacarpal_l {parent=hand_l, loc=<1.7616 -0.9753 3.3814>, rot=<-0.0397 0.0814 -0.069 -0.9935>}
 pinky_01_l {parent=pinky_metacarpal_l, loc=<7.1709 -0.4831 -0.1094>, rot=<-0.0654 -0.0035 0.2242 -0.9723>}
 pinky_02_l {parent=pinky_01_l, loc=<2.8274 0.0893 -0.1043>, rot=<0.0037 0.0011 0.2854 -0.9584>}
 pinky_03_l {parent=pinky_02_l, loc=<2.2046 0.0308 -0.0078>, rot=<-0.0002 0.0098 0.1445 -0.9895>}
 ring_metacarpal_l {parent=hand_l, loc=<2.2074 -0.3625 1.3638>, rot=<-0.0284 0.0722 -0.0458 -0.9959>}
 ring_01_l {parent=ring_metacarpal_l, loc=<7.0295 -0.6197 -0.1446>, rot=<-0.0489 -0.0252 0.1892 -0.9804>}
 ring_02_l {parent=ring_01_l, loc=<3.6174 0.4647 -0.1027>, rot=<0 0 0.2806 -0.9598>}
 ring_03_l {parent=ring_02_l, loc=<2.4152 0.1141 0.0199>, rot=<0 0 0.1189 -0.9929>}
 thumb_01_l {parent=hand_l, loc=<1.3118 -0.3191 -2.459>, rot=<0.6546 -0.4042 0.0664 -0.6354>}
 thumb_02_l {parent=thumb_01_l, loc=<5.9336 -0.0448 0.0401>, rot=<-0.0037 0.0213 0.298 -0.9543>}
 thumb_03_l {parent=thumb_02_l, loc=<2.2223 0.3682 -0.0907>, rot=<0.0118 -0.013 0.0066 -0.9998>}
 lowerarm_twist_01_l {parent=lowerarm_l, loc=<18.316 0 0>, rot=<0 0 0 1>}
 lowerarm_twist_02_l {parent=lowerarm_l, loc=<10.6707 0 0>, rot=<0 0 0 1>}
 dyn_cuffSide_Base_l {parent=lowerarm_twist_02_l, loc=<3.3112 -3.6835 0.1797>, rot=<0.9719 0.2247 0.056 -0.0421>}
 dyn_cuffSide_l {parent=dyn_cuffSide_Base_l, loc=<0 0 0>, rot=<0 0 0 1>}
 dyn_cuffTop_Base_l {parent=lowerarm_twist_02_l, loc=<3.3496 -0.2624 3.8114>, rot=<-0.6963 -0.1855 0.183 -0.6688>}
 dyn_cuffTop_l {parent=dyn_cuffTop_Base_l, loc=<0 0 0>, rot=<0 0 0 -1>}
 upperarm_twist_01_l {parent=upperarm_l, loc=<7.2135 -0.1134 0>, rot=<0 0 0 1>}
 upperarm_twist_02_l {parent=upperarm_l, loc=<19.7998 -0.1134 0>, rot=<0 0 0 1>}
 dyn_collarB_Base_l {parent=clavicle_l, loc=<1.5664 1.0211 7.0059>, rot=<-0.1366 -0.7591 -0.6359 -0.0274>}
 dyn_collarB_l {parent=dyn_collarB_Base_l, loc=<0 0 0>, rot=<0 0 0 1>}
 dyn_collarA_Base_l {parent=clavicle_l, loc=<1.4409 6.9726 2.7188>, rot=<-0.0888 -0.9063 -0.4131 0.0132>}
 dyn_collarA_l {parent=dyn_collarA_Base_l, loc=<0 0 0>, rot=<0 0 0 1>}
 clavicle_r {parent=spine_05, loc=<5.4831 -0.1053 3.2946>, rot=<0.7289 0.0465 -0.6829 0.009>}
 upperarm_r {parent=clavicle_r, loc=<-9.8658 1.2136 0.4104>, rot=<0.0164 -0.3245 -0.0269 -0.9454>}
 lowerarm_r {parent=upperarm_r, loc=<-26.4651 0.8345 -0.259>, rot=<0.0036 0.0161 -0.2298 -0.9731>}
 hand_r {parent=lowerarm_r, loc=<-22.7356 0.3203 0.6106>, rot=<-0.6784 -0.0223 0.0735 -0.7307>}
 index_metacarpal_r {parent=hand_r, loc=<-1.6534 -0.0181 2.9601>, rot=<0.0158 -0.0124 -0.0173 -0.9996>}
 index_01_r {parent=index_metacarpal_r, loc=<-8.3008 -0.0357 0.0524>, rot=<-0.0044 0.0485 0.0924 -0.9945>}
 index_02_r {parent=index_01_r, loc=<-3.3589 0.3003 -0.0453>, rot=<-0.002 0.0027 0.3225 -0.9466>}
 index_03_r {parent=index_02_r, loc=<-2.5578 0.123 0.0252>, rot=<-0.002 0.0067 0.2558 -0.9667>}
 middle_metacarpal_r {parent=hand_r, loc=<-2.1856 0.1726 0.5984>, rot=<0.006 0.0168 -0.0329 -0.9993>}
 middle_01_r {parent=middle_metacarpal_r, loc=<-7.3273 0.2447 -0.0129>, rot=<-0.0208 0.0042 0.0858 -0.9961>}
 middle_02_r {parent=middle_01_r, loc=<-4.0124 -0.14 -0.1175>, rot=<-0.0218 0.0258 0.3512 -0.9357>}
 middle_03_r {parent=middle_02_r, loc=<-2.6077 -0.0533 -0.1373>, rot=<0.0082 0.0188 0.1708 -0.9851>}
 pinky_metacarpal_r {parent=hand_r, loc=<-1.7617 0.9753 -3.3814>, rot=<-0.0397 0.0814 -0.069 -0.9935>}
 pinky_01_r {parent=pinky_metacarpal_r, loc=<-7.1707 0.4832 0.1094>, rot=<-0.0654 -0.0035 0.2242 -0.9723>}
 pinky_02_r {parent=pinky_01_r, loc=<-2.8274 -0.0892 0.1043>, rot=<0.0037 0.0011 0.2854 -0.9584>}
 pinky_03_r {parent=pinky_02_r, loc=<-2.2045 -0.031 0.0078>, rot=<-0.0002 0.0098 0.1445 -0.9895>}
 ring_metacarpal_r {parent=hand_r, loc=<-2.2073 0.3627 -1.3638>, rot=<-0.0284 0.0722 -0.0458 -0.9959>}
 ring_01_r {parent=ring_metacarpal_r, loc=<-7.0297 0.6197 0.1446>, rot=<-0.0489 -0.0252 0.1892 -0.9804>}
 ring_02_r {parent=ring_01_r, loc=<-3.6172 -0.4647 0.1027>, rot=<0 0 0.2806 -0.9598>}
 ring_03_r {parent=ring_02_r, loc=<-2.4152 -0.1141 -0.0199>, rot=<0 0 0.1189 -0.9929>}
 thumb_01_r {parent=hand_r, loc=<-1.3119 0.3192 2.459>, rot=<0.6546 -0.4042 0.0664 -0.6354>}
 thumb_02_r {parent=thumb_01_r, loc=<-5.9336 0.0448 -0.0401>, rot=<-0.0037 0.0213 0.298 -0.9543>}
 thumb_03_r {parent=thumb_02_r, loc=<-2.2223 -0.3682 0.0907>, rot=<0.0118 -0.013 0.0066 -0.9998>}
 lowerarm_twist_01_r {parent=lowerarm_r, loc=<-18.3161 0 0>, rot=<0 0 0 1>}
 lowerarm_twist_02_r {parent=lowerarm_r, loc=<-10.6707 0 0>, rot=<0 0 0 1>}
 dyn_cuffSide_Base_r {parent=lowerarm_twist_02_r, loc=<-3.3109 3.6834 -0.18>, rot=<0.9719 0.2247 0.056 -0.0421>}
 dyn_cuffSide_r {parent=dyn_cuffSide_Base_r, loc=<0 0 0>, rot=<0 0 0 1>}
 dyn_cuffTop_Base_r {parent=lowerarm_twist_02_r, loc=<-3.3495 0.2623 -3.8114>, rot=<-0.6963 -0.1855 0.183 -0.6688>}
 dyn_cuffTop_r {parent=dyn_cuffTop_Base_r, loc=<0 0 0>, rot=<0 0 0 1>}
 upperarm_twist_01_r {parent=upperarm_r, loc=<-7.2135 0.1134 0>, rot=<0 0 0 1>}
 upperarm_twist_02_r {parent=upperarm_r, loc=<-19.7998 0.1134 0>, rot=<0 0 0 1>}
 dyn_collarA_Base_r {parent=clavicle_r, loc=<-1.4409 -6.9726 -2.719>, rot=<-0.0888 -0.9063 -0.4131 0.0132>}
 dyn_collarA_r {parent=dyn_collarA_Base_r, loc=<0 0 0>, rot=<0 0 0 1>}
 dyn_collarB_Base_r {parent=clavicle_r, loc=<-1.5663 -1.0211 -7.0061>, rot=<-0.1366 -0.7591 -0.6359 -0.0274>}
 dyn_collarB_r {parent=dyn_collarB_Base_r, loc=<0 0 0>, rot=<0 0 0 1>}
 neck_01 {parent=spine_05, loc=<11.3138 0.1529 0>, rot=<0 0 -0.2003 -0.9797>}
 neck_02 {parent=neck_01, loc=<4.5627 0 0>, rot=<0 0 0.0504 -0.9987>}
 head {parent=neck_02, loc=<5.3914 0 0>, rot=<0 0 0.115 -0.9934>}
 dyn_camMotion_01_Base {parent=spine_05, loc=<0.6423 9.5213 -0.0003>, rot=<-0.2217 -0.9751 0.0013 0.0041>}
 dyn_camMotion_01 {parent=dyn_camMotion_01_Base, loc=<0 0 0>, rot=<0 0 0 -1>}
 dyn_camMotion_02 {parent=dyn_camMotion_01, loc=<6.0247 0 0>, rot=<-0.0003 -0.0008 0.1432 -0.9897>}
 dyn_camMotion_03 {parent=dyn_camMotion_02, loc=<9.8554 0 0>, rot=<0.0002 0.0018 0.1197 -0.9928>}
 dyn_camStrap_03_r {parent=dyn_camMotion_03, loc=<-0.7089 -2.1021 -3.0205>, rot=<0 0.7095 0 -0.7047>}
 dyn_camStrap_03_l {parent=dyn_camMotion_03, loc=<-0.6018 -2.2079 3.0062>, rot=<0 0.7095 0 -0.7047>}
 dyn_camStrap_01_Base_r {parent=dyn_camMotion_01, loc=<6.1253 0.1751 -3.8278>, rot=<0.0098 0.0246 0.1469 -0.9888>}
 dyn_camStrap_01_r {parent=dyn_camStrap_01_Base_r, loc=<0 0 0>, rot=<0 0 0 1>}
 dyn_camStrap_02_r {parent=dyn_camStrap_01_r, loc=<4.0891 0 0>, rot=<0.0008 0.0324 0.2052 -0.9782>}
 dyn_camStrap_01_Base_l {parent=dyn_camMotion_01, loc=<6.1858 0.2936 3.9718>, rot=<-0.0137 -0.0404 0.1976 -0.9793>}
 dyn_camStrap_01_l {parent=dyn_camStrap_01_Base_l, loc=<0 0 0>, rot=<0 0 0 -1>}
 dyn_camStrap_02_l {parent=dyn_camStrap_01_l, loc=<4.0234 0 0>, rot=<0.0001 -0.0213 0.1348 -0.9906>}
 dyn_collarC_Base_r {parent=spine_05, loc=<13.3481 -3.7102 3.914>, rot=<0.3758 -0.8534 0.1992 0.3013>}
 dyn_collarC_r {parent=dyn_collarC_Base_r, loc=<0 0 0>, rot=<0 0 0 1>}
 dyn_collarBack_Base {parent=spine_05, loc=<13.4111 -4.8047 0>, rot=<0.3456 -0.9382 -0.0164 0.006>}
 dyn_collarBack {parent=dyn_collarBack_Base, loc=<0 0 0>, rot=<0 0 0 1>}
 dyn_collarC_Base_l {parent=spine_05, loc=<13.3482 -3.7102 -3.914>, rot=<0.8534 0.3758 0.3013 -0.1992>}
 dyn_collarC_l {parent=dyn_collarC_Base_l, loc=<0 0 0>, rot=<0 0 0 1>}
 dyn_jacket_bk_01 {parent=spine_04, loc=<-2.4956 -8.767 0>, rot=<-0.0936 -0.9956 0 0>}
 dyn_jacket_bk_02 {parent=dyn_jacket_bk_01, loc=<7.5517 0 0>, rot=<0 0 0.2756 -0.9613>}
 dyn_jacket_bk_03 {parent=dyn_jacket_bk_02, loc=<9.3259 0 0>, rot=<0 0 0 1>}
 dyn_jacket_bk_r_01_Base {parent=spine_04, loc=<-3.0432 -6.5468 8.9164>, rot=<-0.0443 -0.9695 0.2404 -0.0159>}
 dyn_jacket_bk_r_01 {parent=dyn_jacket_bk_r_01_Base, loc=<0 0 0>, rot=<0 0 0 1>}
 dyn_jacket_bk_r_02 {parent=dyn_jacket_bk_r_01, loc=<8.5515 0 0>, rot=<-0.0068 0.0314 0.2479 -0.9683>}
 dyn_jacket_bk_r_03 {parent=dyn_jacket_bk_r_02, loc=<7.1853 0 0>, rot=<-0.0349 0 0 -0.9994>}
 dyn_jacket_bk_l_01_Base {parent=spine_04, loc=<-3.0429 -6.5467 -8.9164>, rot=<0.9695 -0.0443 -0.0159 -0.2404>}
 dyn_jacket_bk_l_01 {parent=dyn_jacket_bk_l_01_Base, loc=<0 0 0>, rot=<0 0 0 1>}
 dyn_jacket_bk_l_02 {parent=dyn_jacket_bk_l_01, loc=<-8.5521 0.0001 0>, rot=<-0.0406 0.0401 0.2467 -0.9674>}
 dyn_jacket_bk_l_03 {parent=dyn_jacket_bk_l_02, loc=<-7.1849 0 0>, rot=<0 0 0 -1>}
 dyn_jacket_r_01_Base {parent=spine_04, loc=<-4.6047 0.0697 11.7338>, rot=<0.0744 -0.7556 0.6508 0.0098>}
 dyn_jacket_r_01 {parent=dyn_jacket_r_01_Base, loc=<0 0 0>, rot=<0 0 0 1>}
 dyn_jacket_r_02 {parent=dyn_jacket_r_01, loc=<8.5184 0 0>, rot=<-0.036 -0.0074 0.09 -0.9953>}
 dyn_jacket_r_03 {parent=dyn_jacket_r_02, loc=<7.4561 0 0>, rot=<0 0 0 -1>}
 dyn_jacket_l_01_Base {parent=spine_04, loc=<-4.6042 0.0697 -11.7338>, rot=<0.7556 0.0744 0.0098 -0.6508>}
 dyn_jacket_l_01 {parent=dyn_jacket_l_01_Base, loc=<0 0 0>, rot=<0 0 0 -1>}
 dyn_jacket_l_02 {parent=dyn_jacket_l_01, loc=<-8.5185 0 0>, rot=<-0.036 -0.0074 0.0899 -0.9953>}
 dyn_jacket_l_03 {parent=dyn_jacket_l_02, loc=<-7.4568 0 0>, rot=<0 0 0 -1>}
 dyn_jacket_fr_r_01_Base {parent=spine_04, loc=<-5.2872 8.9229 9.349>, rot=<0.0621 -0.4103 0.908 -0.0582>}
 dyn_jacket_fr_r_01 {parent=dyn_jacket_fr_r_01_Base, loc=<0 0 0>, rot=<0 0 0 1>}
 dyn_jacket_fr_r_02 {parent=dyn_jacket_fr_r_01, loc=<10.1002 0 0>, rot=<0.0619 0.0239 0.1082 -0.9919>}
 dyn_jacket_fr_r_03 {parent=dyn_jacket_fr_r_02, loc=<7.4598 0 0>, rot=<-0.0872 0 0 -0.9962>}
 dyn_jacket_fr_l_01_Base {parent=spine_04, loc=<-5.2874 8.9229 -9.349>, rot=<0.4103 0.0621 -0.0582 -0.908>}
 dyn_jacket_fr_l_01 {parent=dyn_jacket_fr_l_01_Base, loc=<0 0 0>, rot=<0 0 0 1>}
 dyn_jacket_fr_l_02 {parent=dyn_jacket_fr_l_01, loc=<-10.1003 0 0>, rot=<0.0272 0.0277 0.1073 -0.9935>}
 dyn_jacket_fr_l_03 {parent=dyn_jacket_fr_l_02, loc=<-7.4593 0 0>, rot=<-0.0523 0 0 -0.9986>}
 thigh_l {parent=pelvis, loc=<-1.0855 -0.2109 -8.7053>, rot=<0.0009 0.0355 0.0155 -0.9993>}
 calf_l {parent=thigh_l, loc=<-42.5284 -0.0019 -0.7297>, rot=<0.0006 -0.0161 -0.0391 -0.9991>}
 calf_twist_01_l {parent=calf_l, loc=<-35.8483 -1.9197 -0.5794>, rot=<0.0237 0.0095 -0.0035 -0.9997>}
 dyn_lacesBase_l {parent=calf_twist_01_l, loc=<4.3884 3.91 -1.0171>, rot=<0 0 0 -1>}
 dyn_laceInner_l {parent=dyn_lacesBase_l, loc=<-0.2393 -0.0541 0.3103>, rot=<0.1496 -0.2186 0.9641 0.0206>}
 dyn_laceOuter_l {parent=dyn_lacesBase_l, loc=<-0.5067 -0.1259 -0.5629>, rot=<-0.0849 0.0849 0.9923 0.0311>}
 calf_twist_02_l {parent=calf_l, loc=<-19.1546 -1.7943 -0.2638>, rot=<0.0237 0.0095 -0.0035 -0.9997>}
 foot_l {parent=calf_l, loc=<-46.7401 -2.5014 -0.592>, rot=<0.0427 -0.0238 0.0353 -0.9982>}
 ball_l {parent=foot_l, loc=<-3.5658 10.8066 -0.1544>, rot=<0.0001 -0.0001 -0.7063 -0.7079>}
 thigh_twist_01_l {parent=thigh_l, loc=<-27.952 0 0>, rot=<0.0013 0 0 -1>}
 dyn_hoodieOverleg_01_Base_l {parent=thigh_l, loc=<2.6785 4.1363 -4.0115>, rot=<0.1842 0.3502 0.9106 0.1196>}
 dyn_hoodieOverleg_01_l {parent=dyn_hoodieOverleg_01_Base_l, loc=<0 0 0>, rot=<0 0 0 1>}
 dyn_hoodieOverleg_02_l {parent=dyn_hoodieOverleg_01_l, loc=<6.0424 0 0>, rot=<0 0 0 1>}
 thigh_r {parent=pelvis, loc=<-1.0853 -0.2109 8.7053>, rot=<0.0355 -0.0009 0.9993 0.0155>}
 calf_r {parent=thigh_r, loc=<42.5286 0.0019 0.7297>, rot=<0.0006 -0.0161 -0.0391 -0.9991>}
 calf_twist_01_r {parent=calf_r, loc=<35.8485 1.9197 0.5794>, rot=<0.0237 0.0095 -0.0035 -0.9997>}
 dyn_lacesBase_r {parent=calf_twist_01_r, loc=<-4.3901 -3.8497 1.0664>, rot=<0 0 0 1>}
 dyn_laceOuter_r {parent=dyn_lacesBase_r, loc=<0.5082 0.0656 0.5137>, rot=<0.0849 0.0849 0.0311 -0.9923>}
 dyn_laceInner_r {parent=dyn_lacesBase_r, loc=<0.4165 0.055 -0.588>, rot=<-0.224 -0.138 0.0226 -0.9645>}
 calf_twist_02_r {parent=calf_r, loc=<19.1546 1.7943 0.2638>, rot=<0.0237 0.0095 -0.0035 -0.9997>}
 foot_r {parent=calf_r, loc=<46.7403 2.5014 0.592>, rot=<0.0427 -0.0238 0.0353 -0.9982>}
 ball_r {parent=foot_r, loc=<3.5658 -10.8066 0.1544>, rot=<0.0001 -0.0001 -0.7063 -0.7079>}
 thigh_twist_01_r {parent=thigh_r, loc=<27.952 0 0>, rot=<0.0013 0 0 -1>}
 dyn_pouch_Base_r {parent=thigh_twist_01_r, loc=<-20.217 -2.019 7.0362>, rot=<0.5916 0.0066 0.0092 -0.8062>}
 dyn_pouch_r {parent=dyn_pouch_Base_r, loc=<0 0 0>, rot=<0 0 0 -1>}
 dyn_tassle {parent=dyn_pouch_r, loc=<17.0958 -2.0319 0.1739>, rot=<0.0055 -0.0038 -0.0342 -0.9994>}
 dyn_hoodieKnot {parent=pelvis, loc=<2.2378 9.0787 0.3498>, rot=<0 0 -0.629 -0.7774>}
 dyn_hoodieFront_01_Base_l {parent=dyn_hoodieKnot, loc=<-0.0015 2.0872 -2.2285>, rot=<0.0361 -0.0987 -0.7649 -0.6356>}
 dyn_hoodieFront_01_l {parent=dyn_hoodieFront_01_Base_l, loc=<0 0 0>, rot=<0 0 0 1>}
 dyn_hoodieFront_02_l {parent=dyn_hoodieFront_01_l, loc=<6.8983 0 0>, rot=<-0.0132 -0.0539 0.0228 -0.9982>}
 dyn_hoodieFront_03_l {parent=dyn_hoodieFront_02_l, loc=<6.9026 0 0>, rot=<-0.0821 0.0983 0.0078 -0.9917>}
 dyn_hoodieFront_01_Base_r {parent=dyn_hoodieKnot, loc=<-0.2264 2.722 2.748>, rot=<-0.0334 0.0736 -0.7652 -0.6387>}
 dyn_hoodieFront_01_r {parent=dyn_hoodieFront_01_Base_r, loc=<0 0 0>, rot=<0 0 0 1>}
 dyn_hoodieFront_02_r {parent=dyn_hoodieFront_01_r, loc=<6.4114 0 0>, rot=<0.1099 -0.0118 0.0505 -0.9926>}
 dyn_hoodieFront_03_r {parent=dyn_hoodieFront_02_r, loc=<6.3628 0 0>, rot=<0.0021 0.1007 -0.039 -0.9942>}
 dyn_hoodieOverleg_01_Base_r {parent=pelvis, loc=<3.3644 0.9373 13.1881>, rot=<-0.0543 -0.553 0.8137 0.1708>}
 dyn_hoodieOverleg_01_r {parent=dyn_hoodieOverleg_01_Base_r, loc=<0 0 0>, rot=<0 0 0 -1>}
 dyn_hoodieOverleg_02_r {parent=dyn_hoodieOverleg_01_r, loc=<3.8869 0 0>, rot=<0 0 0.061 -0.9981>}
 dyn_hoodie_01_Base_l {parent=pelvis, loc=<-7.1664 -8.4608 -11.5241>, rot=<-0.0453 -0.9678 -0.2421 -0.0513>}
 dyn_hoodie_01_l {parent=dyn_hoodie_01_Base_l, loc=<0 0 0>, rot=<0 0 0 -1>}
 dyn_hoodie_02_l {parent=dyn_hoodie_01_l, loc=<7.0159 0 0>, rot=<0.0184 -0.0817 0.1062 -0.9908>}
 dyn_hoodie_03_l {parent=dyn_hoodie_02_l, loc=<9.5717 0 0>, rot=<-0.066 -0.028 0.0188 -0.9973>}
 dyn_hoodie_01_c_Base {parent=pelvis, loc=<-9.1694 -10.4561 0.028>, rot=<-0.0039 -1 -0.0001 0.0039>}
 dyn_hoodie_01_c {parent=dyn_hoodie_01_c_Base, loc=<0 0 0>, rot=<0 0 0 -1>}
 dyn_hoodie_02_c {parent=dyn_hoodie_01_c, loc=<7.1432 0 0>, rot=<0.0002 -0.0021 0.063 -0.998>}
 dyn_hoodie_03_c {parent=dyn_hoodie_02_c, loc=<8.629 0 0>, rot=<-0.0016 0.0107 -0.0006 -0.9999>}
 dyn_hoodie_01_Base_r {parent=pelvis, loc=<-7.9154 -10.1619 9.1763>, rot=<-0.0067 -0.9859 0.155 0.0621>}
 dyn_hoodie_01_r {parent=dyn_hoodie_01_Base_r, loc=<0 0 0>, rot=<0 0 0 1>}
 dyn_hoodie_02_r {parent=dyn_hoodie_01_r, loc=<7.1926 0 0>, rot=<0.0657 0.087 0.0238 -0.9937>}
 dyn_hoodie_03_r {parent=dyn_hoodie_02_r, loc=<8.2666 -0.0111 -0.0631>, rot=<-0.0021 0.0231 -0.0113 -0.9997>}
]
variables [
 GoldenGloryEmote {type=double, default=0}
 RigidBody {type=bool, default=False}
 crouching {type=double, default=0}
 isCrouch {type=bool, default=True}
 isFrontEnd {type=bool, default=True}
 isSkydiving {type=bool, default=True}
 noRBAN_Mov {type=bool, default=True}
 noRigidBody {type=bool, default=True}
 notDancing {type=double, default=0}
]
