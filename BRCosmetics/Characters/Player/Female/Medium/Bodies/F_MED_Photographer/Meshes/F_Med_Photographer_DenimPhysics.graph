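# Physics asset for the female medium Photographer body's denim jacket and right-side pouch, giving them secondary motion.
# Jacket capsule chains around back, sides and front hang from small base spheres or spine_02, mostly with tight swing and twist limits.
# Angle-free cross-links join neighboring hem bodies into a skirt-like band open at the front, and the pouch links angle-free to the right thigh.
# Pelvis, spine and thigh capsules approximate the body.
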PhysicsAsset /BRCosmetics/Characters/Player/Female/Medium/Bodies/F_MED_Photographer/Meshes/F_Med_Photographer_DenimPhysics
bodies {
 dyn_jacket_bk_02 [{shapeType=Capsule, length=5, radius=5, shapeName=dyn_jacket_bk_02_capsule, center=<4.6604 -5.1255 0>}]
 dyn_jacket_bk_03 [{shapeType=Capsule, length=5, radius=5, shapeName=dyn_jacket_bk_03_capsule, center=<4.6604 -3.0827 0>}]
 dyn_jacket_bk_l_01_Base [{shapeType=Sphere, radius=1, shapeName=dyn_jacket_bk_l_01_Base_sphere}]
 dyn_jacket_bk_l_02 [{shapeType=Capsule, length=5, radius=5, shapeName=dyn_jacket_bk_l_02_capsule, center=<-5.3277 4.6981 0>}]
 dyn_jacket_bk_l_03 [{shapeType=Capsule, length=5, radius=5, shapeName=dyn_jacket_bk_l_03_capsule, center=<-5.3277 2.6866 0>}]
 dyn_jacket_bk_r_01_Base [{shapeType=Sphere, radius=1, shapeName=dyn_jacket_bk_r_01_Base_sphere}]
 dyn_jacket_bk_r_02 [{shapeType=Capsule, length=5, radius=5, shapeName=dyn_jacket_bk_r_02_capsule, center=<5.153 -3.6832 0>}]
 dyn_jacket_bk_r_03 [{shapeType=Capsule, length=5, radius=5, shapeName=dyn_jacket_bk_r_03_capsule, center=<5.153 -2.5011 0>}]
 dyn_jacket_fr_l_01 [{shapeType=Capsule, length=5, radius=5, shapeName=dyn_jacket_fr_l_02_capsule, center=<-4.8154 1.9067 0>}]
 dyn_jacket_fr_l_01_Base [{shapeType=Sphere, radius=1, shapeName=dyn_jacket_fr_l_01_Base_sphere}]
 dyn_jacket_fr_l_02 [{shapeType=Capsule, length=5, radius=5, shapeName=dyn_jacket_fr_l_02_capsule, center=<-4.8154 3.5029 0>}]
 dyn_jacket_fr_l_03 [{shapeType=Capsule, length=5, radius=5, shapeName=dyn_jacket_fr_l_03_capsule, center=<-5.352 5.6261 0>}]
 dyn_jacket_fr_r_01 [{shapeType=Capsule, length=5, radius=5, shapeName=dyn_jacket_fr_r_02_capsule, center=<5.0768 -2.3684 0>}]
 dyn_jacket_fr_r_01_Base [{shapeType=Sphere, radius=1, shapeName=dyn_jacket_fr_r_01_Base_sphere}]
 dyn_jacket_fr_r_02 [{shapeType=Capsule, length=5, radius=5, shapeName=dyn_jacket_fr_r_02_capsule, center=<5.0768 -4.0135 0>}]
 dyn_jacket_fr_r_03 [{shapeType=Capsule, length=5, radius=5, shapeName=dyn_jacket_fr_r_03_capsule, center=<5.0768 -4.7621 0>}]
 dyn_jacket_l_01_Base [{shapeType=Sphere, radius=1, shapeName=dyn_jacket_l_01_Base_sphere}]
 dyn_jacket_l_02 [{shapeType=Capsule, length=5, radius=5, shapeName=dyn_jacket_l_02_capsule, center=<-5.0873 0 0>}]
 dyn_jacket_l_03 [{shapeType=Capsule, length=5, radius=5, shapeName=dyn_jacket_l_03_capsule, center=<-5.0873 3.8468 0>}]
 dyn_jacket_r_01 [{shapeType=Capsule, length=5, radius=5, shapeName=dyn_jacket_r_02_capsule, center=<5.0768 -2.928 0>}]
 dyn_jacket_r_01_Base [{shapeType=Sphere, radius=1, shapeName=dyn_jacket_r_01_Base_sphere}]
 dyn_jacket_r_02 [{shapeType=Capsule, length=5, radius=5, shapeName=dyn_jacket_r_02_capsule, center=<5.0768 -2.928 0>}]
 dyn_jacket_r_03 [{shapeType=Capsule, length=5, radius=5, shapeName=dyn_jacket_r_03_capsule, center=<5.0768 -2.9373 0>}]
 dyn_pouch_r [{shapeType=Capsule, length=3.7184, radius=6.3108, center=<6.962 -1.3756 0.1274>, rotation=<rot 0 88.6838 -88.8729>}]
 pelvis [
  {shapeType=Capsule, length=7.5341, radius=11.708, center=<0.9996 -0.968 0.0001>}
  {shapeType=Capsule, length=8, radius=4, shapeName=pelvis_capsule, center=<1.9282 16.1342 -2>}
 ]
 spine_01 [{shapeType=Capsule, length=11.1128, radius=8.6096, center=<-0.096 -1.0642 -0.0001>, rotation=<rot 0 10.7679 0>}]
 spine_02 [{shapeType=Capsule, length=5.8895, radius=7.2514, center=<0 2.1568 -0.7518>}]
 thigh_twist_01_l [{shapeType=Capsule, length=29.8822, radius=7.7135, center=<12.39 0.9709 0>, rotation=<rot 90 0 0>}]
 thigh_twist_01_r [{shapeType=Capsule, length=21.4584, radius=7.7135, center=<-12.39 -0.0001 0>, rotation=<rot 90 0 180>}]
}
constraints [
 {bone1Name=dyn_jacket_bk_02, bone2Name=spine_02, swing1LimitDegrees=14, swing2LimitDegrees=10, twistLimitDegrees=4}
 {bone1Name=dyn_jacket_bk_03, bone2Name=dyn_jacket_bk_02, swing1LimitDegrees=28, swing2LimitDegrees=15, twistLimitDegrees=4}
 {bone1Name=dyn_jacket_bk_r_02, bone2Name=dyn_jacket_bk_r_01_Base, swing1LimitDegrees=14, swing2LimitDegrees=10, twistLimitDegrees=4}
 {bone1Name=dyn_jacket_bk_r_03, bone2Name=dyn_jacket_bk_r_02, swing1LimitDegrees=28, swing2LimitDegrees=10, twistLimitDegrees=4}
 {bone1Name=dyn_jacket_r_02, bone2Name=dyn_jacket_r_01, swing1LimitDegrees=14, swing2LimitDegrees=14, twistLimitDegrees=4}
 {bone1Name=dyn_jacket_r_03, bone2Name=dyn_jacket_r_02, swing1LimitDegrees=28, swing2LimitDegrees=14, twistLimitDegrees=4}
 {bone1Name=dyn_jacket_fr_r_02, bone2Name=dyn_jacket_fr_r_01, swing1LimitDegrees=28, swing2LimitDegrees=20, twistLimitDegrees=4}
 {bone1Name=dyn_jacket_fr_r_03, bone2Name=dyn_jacket_fr_r_02, swing1LimitDegrees=28, swing2LimitDegrees=10, twistLimitDegrees=4}
 {bone1Name=dyn_jacket_fr_l_02, bone2Name=dyn_jacket_fr_l_01, swing1LimitDegrees=28, swing2LimitDegrees=20, twistLimitDegrees=4}
 {bone1Name=dyn_jacket_fr_l_03, bone2Name=dyn_jacket_fr_l_02, swing1LimitDegrees=28, swing2LimitDegrees=20, twistLimitDegrees=4}
 {bone1Name=dyn_jacket_l_02, bone2Name=dyn_jacket_l_01_Base, swing1LimitDegrees=14, swing2LimitDegrees=14, twistLimitDegrees=4}
 {bone1Name=dyn_jacket_l_03, bone2Name=dyn_jacket_l_02, swing1LimitDegrees=28, swing2LimitDegrees=14, twistLimitDegrees=4}
 {bone1Name=dyn_jacket_bk_l_02, bone2Name=dyn_jacket_bk_l_01_Base, swing1LimitDegrees=14, swing2LimitDegrees=10, twistLimitDegrees=4}
 {bone1Name=dyn_jacket_bk_l_03, bone2Name=dyn_jacket_bk_l_02, swing1LimitDegrees=28, swing2LimitDegrees=10, twistLimitDegrees=4}
 {
  bone1Name=dyn_pouch_r
  bone2Name=thigh_twist_01_r
  swing1LimitDegrees=45
  swing1Motion=Free
  swing2LimitDegrees=45
  swing2Motion=Free
  twistLimitDegrees=45
  twistMotion=Free
 }
 {
  bone1Name=dyn_jacket_bk_r_03
  bone2Name=dyn_jacket_bk_03
  swing1LimitDegrees=45
  swing1Motion=Free
  swing2LimitDegrees=45
  swing2Motion=Free
  twistLimitDegrees=45
  twistMotion=Free
 }
 {
  bone1Name=dyn_jacket_bk_l_03
  bone2Name=dyn_jacket_bk_03
  swing1LimitDegrees=45
  swing1Motion=Free
  swing2LimitDegrees=45
  swing2Motion=Free
  twistLimitDegrees=45
  twistMotion=Free
 }
 {
  bone1Name=dyn_jacket_bk_r_03
  bone2Name=dyn_jacket_r_03
  swing1LimitDegrees=45
  swing1Motion=Free
  swing2LimitDegrees=45
  swing2Motion=Free
  twistLimitDegrees=45
  twistMotion=Free
 }
 {
  bone1Name=dyn_jacket_r_03
  bone2Name=dyn_jacket_fr_r_03
  swing1LimitDegrees=45
  swing1Motion=Free
  swing2LimitDegrees=45
  swing2Motion=Free
  twistLimitDegrees=45
  twistMotion=Free
 }
 {
  bone1Name=dyn_jacket_l_03
  bone2Name=dyn_jacket_bk_l_03
  swing1LimitDegrees=45
  swing1Motion=Free
  swing2LimitDegrees=45
  swing2Motion=Free
  twistLimitDegrees=45
  twistMotion=Free
 }
 {
  bone1Name=dyn_jacket_fr_l_03
  bone2Name=dyn_jacket_l_03
  swing1LimitDegrees=45
  swing1Motion=Free
  swing2LimitDegrees=45
  swing2Motion=Free
  twistLimitDegrees=45
  twistMotion=Free
 }
 {bone1Name=dyn_jacket_fr_r_01, bone2Name=dyn_jacket_fr_r_01_Base, swing1LimitDegrees=25, swing2LimitDegrees=28, twistLimitDegrees=45}
 {bone1Name=dyn_jacket_fr_l_01, bone2Name=dyn_jacket_fr_l_01_Base, swing1LimitDegrees=25, swing2LimitDegrees=28, twistLimitDegrees=45}
 {
  bone1Name=dyn_jacket_r_02
  bone2Name=dyn_jacket_bk_r_02
  swing1LimitDegrees=45
  swing1Motion=Free
  swing2LimitDegrees=45
  swing2Motion=Free
  twistLimitDegrees=45
  twistMotion=Free
 }
 {
  bone1Name=dyn_jacket_r_01
  bone2Name=dyn_jacket_r_01_Base
  swing1LimitDegrees=45
  swing1Motion=Free
  swing2LimitDegrees=45
  swing2Motion=Free
  twistLimitDegrees=45
  twistMotion=Free
 }
]
