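# Physics asset for the female medium Photographer body, covering secondary motion of the hoodie and a camMotion chain (likely the camera).
# Three twist-locked hoodie chains (left, center, right) and two front chains each link three capsules to a small base sphere with limited swing.
# The tightly limited camMotion chain joins spine_05 through an angularly free base.
# Pelvis, upper spine and thigh-twist capsules form the body shapes around the chains.
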PhysicsAsset /BRCosmetics/Characters/Player/Female/Medium/Bodies/F_MED_Photographer/Meshes/F_Med_Photographer_HoodiePhysics
bodies {
 dyn_camMotion_01 [{shapeType=Capsule, length=6.75, radius=3.75, shapeName=dyn_camMotion_01_capsule, center=<3.0645 3.7519 0.0001>}]
 dyn_camMotion_01_Base [{shapeType=Sphere, radius=0.505, shapeName=dyn_camMotion_01_Base_sphere}]
 dyn_camMotion_02 [{shapeType=Capsule, length=5.75, radius=4, shapeName=dyn_camMotion_02_capsule, center=<4.8667 3.3635 0.0005>}]
 dyn_camMotion_03 [{shapeType=Capsule, length=7.25, radius=5.5, shapeName=dyn_camMotion_02_capsule, center=<5.6822 3.1923 0.5323>}]
 dyn_hoodieFront_01_Base_l [{shapeType=Sphere, radius=1, shapeName=dyn_hoodieFront_01_Base_l_sphere}]
 dyn_hoodieFront_01_Base_r [{shapeType=Sphere, radius=1, shapeName=dyn_hoodieFront_01_Base_r_sphere}]
 dyn_hoodieFront_01_l [{shapeType=Capsule, length=2, radius=3, shapeName=dyn_hoodieFront_01_l_capsule, center=<0.6077 -1.9447 0>}]
 dyn_hoodieFront_01_r [{shapeType=Capsule, length=2, radius=3, shapeName=dyn_hoodieFront_01_r_capsule, center=<0 -1.553 0>}]
 dyn_hoodieFront_02_l [{shapeType=Capsule, length=2, radius=3, shapeName=dyn_hoodieFront_02_l_capsule, center=<0 -1.9447 0>}]
 dyn_hoodieFront_02_r [{shapeType=Capsule, length=2, radius=3, shapeName=dyn_hoodieFront_02_r_capsule, center=<0 -1.553 0>}]
 dyn_hoodieFront_03_l [{shapeType=Capsule, length=2, radius=4, shapeName=dyn_hoodieFront_03_l_capsule, center=<0 -1.9447 0>}]
 dyn_hoodieFront_03_r [{shapeType=Capsule, length=2, radius=4, shapeName=dyn_hoodieFront_03_r_capsule, center=<0 -1.553 0>}]
 dyn_hoodie_01_Base_l [{shapeType=Sphere, radius=1, shapeName=dyn_hoodie_01_Base_l_sphere}]
 dyn_hoodie_01_Base_r [{shapeType=Sphere, radius=1, shapeName=dyn_hoodie_01_Base_r_sphere}]
 dyn_hoodie_01_c [{shapeType=Capsule, length=7, radius=4, shapeName=dyn_hoodie_01_c_capsule, center=<3.3531 -2.6422 1.2282>}]
 dyn_hoodie_01_c_Base [{shapeType=Sphere, radius=1, shapeName=dyn_hoodie_01_c_Base_sphere}]
 dyn_hoodie_01_l [{shapeType=Capsule, length=7.25, radius=4, shapeName=dyn_hoodie_01_l_capsule, center=<3.6208 -4.2309 1.5177>, rotation=<rot -10 0 0>}]
 dyn_hoodie_01_r [{shapeType=Capsule, length=4, radius=4, shapeName=dyn_hoodie_01_r_capsule, center=<4.5369 -1.5668 -0.9965>, rotation=<rot 10 0 0>}]
 dyn_hoodie_02_c [{shapeType=Capsule, length=7.75, radius=4.25, shapeName=dyn_hoodie_02_c_capsule, center=<4.317 -2.6422 1.3546>}]
 dyn_hoodie_02_l [{shapeType=Capsule, length=9, radius=4.2, shapeName=dyn_hoodie_02_l_capsule, center=<5.9742 -4.2163 1.6838>, rotation=<rot 0 0 10.0003>}]
 dyn_hoodie_02_r [{shapeType=Capsule, length=4, radius=4, shapeName=dyn_hoodie_02_r_capsule, center=<5.5222 -1.5855 -1.2467>}]
 dyn_hoodie_03_c [{shapeType=Capsule, length=8, radius=4.4, shapeName=dyn_hoodie_03_c_capsule, center=<3.3531 -2.6422 0.8849>}]
 dyn_hoodie_03_l [{shapeType=Capsule, length=9, radius=4.2, shapeName=dyn_hoodie_03_l_capsule, center=<4.268 -1.1246 2.6088>, rotation=<rot 0 0 10.0011>}]
 dyn_hoodie_03_r [{shapeType=Capsule, length=4, radius=4, shapeName=dyn_hoodie_03_r_capsule, center=<4.8978 -1.6765 -0.6151>}]
 pelvis [{shapeType=Capsule, length=15.9671, radius=6.6828, center=<-8.3559 1.5344 0>}]
 spine_03 [{shapeType=Capsule, length=0.8477, radius=10.5782, center=<0 0.6736 0>}]
 spine_04 [{shapeType=Capsule, length=17.2988, radius=7.8624, center=<-1.0998 4.0228 0>}]
 spine_05 [{shapeType=Capsule, length=15.1462, radius=6.6835, center=<-4.9741 3.4142 -0.0001>}]
 thigh_twist_01_l [{shapeType=Capsule, length=34, radius=7.175, center=<7 -1.9664 0.6373>, rotation=<rot 90 0 0>}]
 thigh_twist_01_r [{shapeType=Capsule, length=22.8692, radius=7.5798, center=<-14.7835 -0.0001 0>, rotation=<rot 90 0 0>}]
}
constraints [
 {
  bone1Name=dyn_hoodie_01_l
  bone2Name=dyn_hoodie_01_Base_l
  swing1LimitDegrees=40
  swing2LimitDegrees=16
  twistLimitDegrees=45
  twistMotion=Locked
 }
 {bone1Name=dyn_hoodie_02_l, bone2Name=dyn_hoodie_01_l, swing1LimitDegrees=30, swing2LimitDegrees=12, twistLimitDegrees=45, twistMotion=Locked}
 {bone1Name=dyn_hoodie_03_l, bone2Name=dyn_hoodie_02_l, swing1LimitDegrees=30, swing2LimitDegrees=12, twistLimitDegrees=10, twistMotion=Locked}
 {
  bone1Name=dyn_hoodie_01_c
  bone2Name=dyn_hoodie_01_c_Base
  swing1LimitDegrees=40
  swing2LimitDegrees=16
  twistLimitDegrees=45
  twistMotion=Locked
 }
 {bone1Name=dyn_hoodie_02_c, bone2Name=dyn_hoodie_01_c, swing1LimitDegrees=30, swing2LimitDegrees=12, twistLimitDegrees=45, twistMotion=Locked}
 {bone1Name=dyn_hoodie_03_c, bone2Name=dyn_hoodie_02_c, swing1LimitDegrees=30, swing2LimitDegrees=12, twistLimitDegrees=45, twistMotion=Locked}
 {
  bone1Name=dyn_hoodie_01_r
  bone2Name=dyn_hoodie_01_Base_r
  swing1LimitDegrees=40
  swing2LimitDegrees=16
  twistLimitDegrees=45
  twistMotion=Locked
 }
 {bone1Name=dyn_hoodie_02_r, bone2Name=dyn_hoodie_01_r, swing1LimitDegrees=30, swing2LimitDegrees=12, twistLimitDegrees=45, twistMotion=Locked}
 {bone1Name=dyn_hoodie_03_r, bone2Name=dyn_hoodie_02_r, swing1LimitDegrees=30, swing2LimitDegrees=12, twistLimitDegrees=45, twistMotion=Locked}
 {bone1Name=dyn_hoodieFront_01_l, bone2Name=dyn_hoodieFront_01_Base_l, swing1LimitDegrees=25, swing2LimitDegrees=25, twistLimitDegrees=18}
 {bone1Name=dyn_hoodieFront_02_l, bone2Name=dyn_hoodieFront_01_l, swing1LimitDegrees=25, swing2LimitDegrees=25, twistLimitDegrees=18}
 {bone1Name=dyn_hoodieFront_03_l, bone2Name=dyn_hoodieFront_02_l, swing1LimitDegrees=25, swing2LimitDegrees=25, twistLimitDegrees=18}
 {bone1Name=dyn_hoodieFront_01_r, bone2Name=dyn_hoodieFront_01_Base_r, swing1LimitDegrees=25, swing2LimitDegrees=25, twistLimitDegrees=18}
 {bone1Name=dyn_hoodieFront_02_r, bone2Name=dyn_hoodieFront_01_r, swing1LimitDegrees=25, swing2LimitDegrees=25, twistLimitDegrees=18}
 {bone1Name=dyn_hoodieFront_03_r, bone2Name=dyn_hoodieFront_02_r, swing1LimitDegrees=25, swing2LimitDegrees=25, twistLimitDegrees=18}
 {
  bone1Name=spine_04
  bone2Name=spine_03
  swing1LimitDegrees=45
  swing1Motion=Free
  swing2LimitDegrees=45
  swing2Motion=Free
  twistLimitDegrees=45
  twistMotion=Free
 }
 {
  bone1Name=spine_05
  bone2Name=spine_04
  swing1LimitDegrees=45
  swing1Motion=Free
  swing2LimitDegrees=45
  swing2Motion=Free
  twistLimitDegrees=45
  twistMotion=Free
 }
 {
  bone1Name=dyn_camMotion_01_Base
  bone2Name=spine_05
  swing1LimitDegrees=45
  swing1Motion=Free
  swing2LimitDegrees=45
  swing2Motion=Free
  twistLimitDegrees=45
  twistMotion=Free
 }
 {bone1Name=dyn_camMotion_01, bone2Name=dyn_camMotion_01_Base, swing1LimitDegrees=15, swing2LimitDegrees=15, twistLimitDegrees=2}
 {bone1Name=dyn_camMotion_02, bone2Name=dyn_camMotion_01, swing1LimitDegrees=15, swing2LimitDegrees=15, twistLimitDegrees=2}
 {bone1Name=dyn_camMotion_03, bone2Name=dyn_camMotion_02, swing1LimitDegrees=15, swing2LimitDegrees=15, twistLimitDegrees=2}
]
