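# Anim Blueprint for the F_MED_Photographer cosmetic body, adding secondary motion and corrective posing to the jacket, hoodie and accessories.
# The pose copied from the attached parent runs through a jacket Control Rig, then one of two denim RigidBody setups chosen by backbling or cape.
# Hoodie RigidBody, AnimDynamics for pouch, tassel, laces and cuffs, and bone offsets follow.
# Thigh and spine twist, skydiving and Golden Glory emote detection drive the corrections.
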AnimBlueprint /BRCosmetics/Characters/Player/Female/Medium/Bodies/F_MED_Photographer/Meshes/F_MED_Photographer_AnimBP
target_skeleton=&/F_MED_Photographer_Skeleton
graphs [
 AnimGraph {nodes [
   n1 = Look At - Bone: dyn_jacket_fr_l_01_Base(
    ComponentPose=Local To Component(LocalPose=Use cached pose 'Photog'())
    Alpha=Get JacketFrontSideCorrection() -> (JacketFrontSideCorrection:double)
   ) {
    bUseLookUpAxis=true
    alphaBoolBlend {blendOption=Linear}
    boneToModify {boneName=dyn_jacket_fr_l_01_Base}
    lookAtLocation=<0 0 30>
    lookAtTarget {boneReference {boneName=thigh_l}}
    lookAt_Axis {axis=<1 0>}
   }
   n2 = RigidBody(ComponentPose=@n1) {
    bClampLinearTranslationLimitToRefPose=true
    bOverrideWorldGravity=true
    overridePhysicsAsset=&/F_Med_Photographer_DenimPhysics
    simulationSpace=BaseBoneSpace
    alphaBoolBlend {blendOption=Linear}
    baseBoneRef {boneName=pelvis}
    componentAppliedLinearAccClamp=<3000 3000 3000>
    componentLinearAccScale=<0.3 0.2 0.2>
    componentLinearVelScale=<0.3 0.2 0.2>
    overrideWorldGravity {z=-980}
    simSpaceSettings {maxAngularAcceleration=1000, maxAngularVelocity=3, maxLinearAcceleration=250, maxLinearVelocity=500, velocityScaleZ=0, worldAlpha=1}
   }
   n3 = Component To Local(ComponentPose=@n2)
   n4 = Look At - Bone: dyn_jacket_fr_l_01_Base(ComponentPose=Local To Component(LocalPose=Use cached pose 'Photog'()), Alpha=Get JacketFrontSideCorrection()) {
    bUseLookUpAxis=true
    alphaBoolBlend {blendOption=Linear}
    boneToModify {boneName=dyn_jacket_fr_l_01_Base}
    lookAtLocation {x=0, z=30}
    lookAtTarget {boneReference {boneName=thigh_l}}
    lookAt_Axis {axis=<1 0>}
   }
   n5 = RigidBody(ComponentPose=@n4) {
    bClampLinearTranslationLimitToRefPose=true
    bOverrideWorldGravity=true
    overridePhysicsAsset=&/F_Med_Photographer_DenimBackBlingPhysics
    simulationSpace=BaseBoneSpace
    alphaBoolBlend {blendOption=Linear}
    baseBoneRef {boneName=pelvis}
    componentAppliedLinearAccClamp=<300 300 300>
    componentLinearAccScale=<0.2 0.2 0.2>
    componentLinearVelScale=<0.2 0.2 0.2>
    overrideWorldGravity {z=-980}
    simSpaceSettings {maxAngularAcceleration=1000, maxAngularVelocity=3, maxLinearAcceleration=250, maxLinearVelocity=500, velocityScaleZ=0, worldAlpha=1}
   }
   n6 = Component To Local(ComponentPose=@n5)
   n7 = Get Backbling() -> (Backbling:bool)
   n8 = RigidBody!(
    ComponentPose=Local To Component(LocalPose=Blend Poses by bool(BlendPose_0=@n6, BlendPose_1=@n3, bActiveValue=@n7) {blendType=Linear})
    Alpha
    OverrideWorldGravity:Vector="0,0,-980"
    ExternalForce
    ComponentLinearAccScale:Vector="0.3,0.2,0.2"
    ComponentLinearVelScale:Vector="0.3,0.2,0.2"
   ) -> (Pose) {
    bClampLinearTranslationLimitToRefPose=true
    bOverrideWorldGravity=true
    overridePhysicsAsset=&/F_Med_Photographer_HoodiePhysics
    simulationSpace=BaseBoneSpace
    alphaBoolBlend {blendOption=Linear}
    alphaScaleBiasClamp {clampMin=0.4}
    baseBoneRef {boneName=pelvis}
    componentAppliedLinearAccClamp=<1500 1500 3000>
    componentLinearAccScale=<0.4 0.3 0.3>
    componentLinearVelScale=<0.4 0.3 0.3>
    overrideWorldGravity {z=-300}
    simSpaceSettings {maxAngularAcceleration=3000, maxAngularVelocity=3, maxLinearAcceleration=250, maxLinearVelocity=500, velocityScaleZ=0, worldAlpha=1}
   }
   n9 = "Transform (Modify) Bone - Bone: dyn_hoodieFront_01_l"!(ComponentPose=@n8, bAlphaBoolEnabled:bool=True, Translation="0,0,0", Rotation="20,15,0", Scale="1,1,1") -> (Pose) {
    alphaInputType=Bool
    rotationMode=BMM_Additive
    rotationSpace=BCS_BoneSpace
    alphaBoolBlend {blendOption=Linear}
    boneToModify {boneName=dyn_hoodieFront_01_l}
    rotation {pitch=20, yaw=15}
   }
   n10 = "Transform (Modify) Bone - Bone: dyn_hoodieFront_01_r"!(ComponentPose=@n9, bAlphaBoolEnabled:bool=True, Translation="0,0,0", Rotation="-10,0,0", Scale="1,1,1") -> (Pose) {
    alphaInputType=Bool
    rotationMode=BMM_Additive
    rotationSpace=BCS_BoneSpace
    alphaBoolBlend {blendOption=Linear}
    boneToModify {boneName=dyn_hoodieFront_01_r}
    rotation {pitch=-10}
   }
   n11 = Anim Dynamics - Bone: dyn_pouch_r(ComponentPose=@n10, Alpha=1, ExternalForce="0,0,0") {
    bUseGravityOverride=true
    simulationSpace=RootRelative
    alphaBoolBlend {blendOption=Linear}
    boundBone {boneName=dyn_pouch_r}
    componentAppliedLinearAccClamp=<3000 3000 300>
    componentLinearAccScale=<0.3 0.2 0.2>
    componentLinearVelScale=<0.3 0.2 0.2>
    gravityOverride {z=-980}
    physicsBodyDefinitions [{
      boundBone {boneName=dyn_pouch_r}
      boxExtents=<30 20 20>
      constraintSetup {angularLimitsMax=<2 5 1>, angularLimitsMin=<-2 -5 -10>, angularTarget=<1 0 0>}
      localJointOffset=<15 -2 0>
     }]
   }
   n12 = Anim Dynamics - Bone: dyn_tassle(ComponentPose=@n11, Alpha=1, ExternalForce="0,0,0") {
    bUseGravityOverride=true
    simulationSpace=RootRelative
    alphaBoolBlend {blendOption=Linear}
    boundBone {boneName=dyn_tassle}
    componentAppliedLinearAccClamp=<3000 3000 300>
    componentLinearAccScale=<0.3 0.2 0.2>
    componentLinearVelScale=<0.3 0.2 0.2>
    gravityOverride {z=-980}
    physicsBodyDefinitions [{
      boundBone {boneName=dyn_tassle}
      boxExtents=<10 10 10>
      constraintSetup {angularLimitsMax=<20 20 20>, angularLimitsMin=<-20 -20 -20>, angularTarget=<1 0 0>}
      localJointOffset=<5 0 0>
     }]
   }
   n13 = Anim Dynamics - Bone: dyn_laceInner_l(ComponentPose=@n12, Alpha=1, ExternalForce="0,0,0") {
    bUseGravityOverride=true
    simulationSpace=RootRelative
    alphaBoolBlend {blendOption=Linear}
    boundBone {boneName=dyn_laceInner_l}
    componentAppliedLinearAccClamp=<3000 3000 300>
    componentLinearAccScale=<0.3 0.2 0.2>
    componentLinearVelScale=<0.3 0.2 0.2>
    gravityOverride {z=-980}
    physicsBodyDefinitions [{
      boundBone {boneName=dyn_laceInner_l}
      boxExtents=<10 5 5>
      constraintSetup {angularLimitsMax=<10 10 20>, angularLimitsMin=<-10 -10 -20>, angularTarget=<1 0 0>}
      localJointOffset=<5 0 0>
     }]
   }
   n14 = Anim Dynamics - Bone: dyn_laceOuter_r(ComponentPose=@n13, Alpha=1, ExternalForce="0,0,0") {
    bUseGravityOverride=true
    simulationSpace=RootRelative
    alphaBoolBlend {blendOption=Linear}
    boundBone {boneName=dyn_laceOuter_r}
    componentAppliedLinearAccClamp=<3000 3000 300>
    componentLinearAccScale=<0.3 0.2 0.2>
    componentLinearVelScale=<0.3 0.2 0.2>
    gravityOverride {z=-980}
    physicsBodyDefinitions [{
      boundBone {boneName=dyn_laceOuter_r}
      boxExtents=<10 5 5>
      constraintSetup {angularLimitsMax=<10 10 20>, angularLimitsMin=<-10 -10 -20>, angularTarget=<1 0 0>}
      localJointOffset=<5 0 0>
     }]
   }
   n15 = Anim Dynamics - Bone: dyn_laceInner_r(ComponentPose=@n14, Alpha=1, ExternalForce="0,0,0") {
    bUseGravityOverride=true
    simulationSpace=RootRelative
    alphaBoolBlend {blendOption=Linear}
    boundBone {boneName=dyn_laceInner_r}
    componentAppliedLinearAccClamp=<3000 3000 300>
    componentLinearAccScale=<0.3 0.2 0.2>
    componentLinearVelScale=<0.3 0.2 0.2>
    gravityOverride {z=-980}
    physicsBodyDefinitions [{
      boundBone {boneName=dyn_laceInner_r}
      boxExtents=<10 5 5>
      constraintSetup {angularLimitsMax=<10 10 20>, angularLimitsMin=<-10 -10 -20>, angularTarget=<1 0 0>}
      localJointOffset=<5 0 0>
     }]
   }
   n16 = Anim Dynamics - Bone: dyn_laceOuter_l(ComponentPose=@n15, Alpha=1, ExternalForce="0,0,0") {
    bUseGravityOverride=true
    simulationSpace=RootRelative
    alphaBoolBlend {blendOption=Linear}
    boundBone {boneName=dyn_laceOuter_l}
    componentAppliedLinearAccClamp=<3000 3000 300>
    componentLinearAccScale=<0.3 0.2 0.2>
    componentLinearVelScale=<0.3 0.2 0.2>
    gravityOverride {z=-980}
    physicsBodyDefinitions [{
      boundBone {boneName=dyn_laceOuter_l}
      boxExtents=<10 5 5>
      constraintSetup {angularLimitsMax=<10 10 20>, angularLimitsMin=<-10 -10 -20>, angularTarget=<1 0 0>}
      localJointOffset=<5 0 0>
     }]
   }
   n17 = Anim Dynamics - Bone: dyn_cuffSide_r(ComponentPose=@n16, Alpha=1, ExternalForce="0,0,0") {
    bUseGravityOverride=true
    simulationSpace=RootRelative
    alphaBoolBlend {blendOption=Linear}
    boundBone {boneName=dyn_cuffSide_r}
    componentAppliedLinearAccClamp=<3000 3000 300>
    componentLinearAccScale=<0.3 0.2 0.2>
    componentLinearVelScale=<0.3 0.2 0.2>
    gravityOverride {z=-980}
    physicsBodyDefinitions [{
      boundBone {boneName=dyn_cuffSide_r}
      boxExtents=<5 5 5>
      constraintSetup {angularLimitsMax=<5 5 5>, angularLimitsMin=<-5 -5 -5>, angularTarget=<1 0 0>}
      localJointOffset=<2.5 0 0>
     }]
   }
   n18 = Anim Dynamics - Bone: dyn_cuffTop_r(ComponentPose=@n17, Alpha=1, ExternalForce="0,0,0") {
    bUseGravityOverride=true
    simulationSpace=RootRelative
    alphaBoolBlend {blendOption=Linear}
    boundBone {boneName=dyn_cuffTop_r}
    componentAppliedLinearAccClamp=<3000 3000 300>
    componentLinearAccScale=<0.3 0.2 0.2>
    componentLinearVelScale=<0.3 0.2 0.2>
    gravityOverride {z=-980}
    physicsBodyDefinitions [{
      boundBone {boneName=dyn_cuffTop_r}
      boxExtents=<5 5 5>
      constraintSetup {angularLimitsMax=<5 5 5>, angularLimitsMin=<-5 -5 -5>, angularTarget=<1 0 0>}
      localJointOffset=<2.5 0 0>
     }]
   }
   n19 = Anim Dynamics - Bone: dyn_cuffTop_l(ComponentPose=@n18, Alpha=1, ExternalForce="0,0,0") {
    bUseGravityOverride=true
    simulationSpace=RootRelative
    alphaBoolBlend {blendOption=Linear}
    boundBone {boneName=dyn_cuffTop_l}
    componentAppliedLinearAccClamp=<3000 3000 300>
    componentLinearAccScale=<0.3 0.2 0.2>
    componentLinearVelScale=<0.3 0.2 0.2>
    gravityOverride {z=-980}
    physicsBodyDefinitions [{
      boundBone {boneName=dyn_cuffTop_l}
      boxExtents=<5 5 5>
      constraintSetup {angularLimitsMax=<5 5 5>, angularLimitsMin=<-5 -5 -5>, angularTarget=<1 0 0>}
      localJointOffset=<-2.5 0 0>
     }]
   }
   n20 = Anim Dynamics - Bone: dyn_cuffSide_l(ComponentPose=@n19, Alpha=1, ExternalForce="0,0,0") {
    bUseGravityOverride=true
    simulationSpace=RootRelative
    alphaBoolBlend {blendOption=Linear}
    boundBone {boneName=dyn_cuffSide_l}
    componentAppliedLinearAccClamp=<3000 3000 300>
    componentLinearAccScale=<0.3 0.2 0.2>
    componentLinearVelScale=<0.3 0.2 0.2>
    gravityOverride {z=-980}
    physicsBodyDefinitions [{
      boundBone {boneName=dyn_cuffSide_l}
      boxExtents=<5 5 5>
      constraintSetup {angularLimitsMax=<5 5 5>, angularLimitsMin=<-5 -5 -5>, angularTarget=<1 0 0>}
      localJointOffset=<-2.5 0 0>
     }]
   }
   n21 = Get SkydiveAlpha() -> (SkydiveAlpha:float)
   n22 = "Transform (Modify) Bone - Bone: dyn_jacket_bk_r_01_Base"(ComponentPose=@n20, Alpha=@n21, Translation="0,0,0", Rotation="0.000,-12,0.000", Scale="1,1,1") {
    rotationMode=BMM_Additive
    rotationSpace=BCS_BoneSpace
    translationMode=BMM_Additive
    translationSpace=BCS_BoneSpace
    alphaBoolBlend {blendInTime=0.5, blendOption=Linear, blendOutTime=0.1}
    boneToModify {boneName=dyn_jacket_bk_r_01_Base}
    rotation {yaw=-12}
   }
   n23 = "Transform (Modify) Bone - Bone: dyn_jacket_bk_l_01_Base"(ComponentPose=@n22, Alpha=@n21, Translation="0,0,0", Rotation="0,-12,0", Scale="1,1,1") {
    rotationMode=BMM_Additive
    rotationSpace=BCS_BoneSpace
    translationMode=BMM_Additive
    translationSpace=BCS_BoneSpace
    alphaBoolBlend {blendInTime=0.2, blendOption=Linear, blendOutTime=0.1}
    boneToModify {boneName=dyn_jacket_bk_l_01_Base}
    rotation {yaw=-12}
   }
   n24 = "Transform (Modify) Bone - Bone: dyn_jacket_bk_01"(ComponentPose=@n23, Alpha=@n21, Translation="0,0,0", Rotation="0,-13.5,-0.000", Scale="1,1,1") {
    rotationMode=BMM_Additive
    rotationSpace=BCS_BoneSpace
    translationMode=BMM_Additive
    translationSpace=BCS_BoneSpace
    alphaBoolBlend {blendInTime=0.5, blendOption=Linear, blendOutTime=0.1}
    boneToModify {boneName=dyn_jacket_bk_01}
    rotation {yaw=-13.5}
   }
   n25 = "Transform (Modify) Bone - Bone: dyn_jacket_r_01_Base"(ComponentPose=@n24, Alpha=@n21, Translation="0,0,0", Rotation="0.000,-12,0.000", Scale="1,1,1") {
    rotationMode=BMM_Additive
    rotationSpace=BCS_BoneSpace
    translationMode=BMM_Additive
    translationSpace=BCS_BoneSpace
    alphaBoolBlend {blendInTime=0.5, blendOption=Linear, blendOutTime=0.1}
    boneToModify {boneName=dyn_jacket_r_01_Base}
    rotation {yaw=-12}
   }
   n26 = "Transform (Modify) Bone - Bone: dyn_jacket_l_01_Base"(ComponentPose=@n25, Alpha=@n21, Translation="0,0,0", Rotation="0,-12,0", Scale="1,1,1") {
    rotationMode=BMM_Additive
    rotationSpace=BCS_BoneSpace
    translationMode=BMM_Additive
    translationSpace=BCS_BoneSpace
    alphaBoolBlend {blendInTime=0.2, blendOption=Linear, blendOutTime=0.1}
    boneToModify {boneName=dyn_jacket_l_01_Base}
    rotation {yaw=-12}
   }
   n27 = "Transform (Modify) Bone - Bone: dyn_jacket_bk_01"(ComponentPose=@n26, Alpha=Get WeaponChange() -> (WeaponChange:double), Translation="0,0,0", Rotation="0,0,0", Scale="1,3,1") {
    rotationSpace=BCS_BoneSpace
    scaleMode=BMM_Replace
    scaleSpace=BCS_BoneSpace
    translationSpace=BCS_BoneSpace
    alphaBoolBlend {blendOption=Linear}
    alphaScaleBiasClamp {bInterpResult=true, interpSpeedDecreasing=2}
    boneToModify {boneName=dyn_jacket_bk_01}
    scale {y=3}
   }
   n28 = Get JacketFrontSideCorrection()
   n29 = "Transform (Modify) Bone - Bone: dyn_jacket_fr_l_01_Base"(ComponentPose=@n27, Alpha=@n28, Translation="0,0,0", Rotation="-0.000,-20.000,-0.000", Scale="1,1,1") {
    rotationMode=BMM_Additive
    rotationSpace=BCS_BoneSpace
    scaleSpace=BCS_BoneSpace
    translationSpace=BCS_BoneSpace
    alphaBoolBlend {blendOption=Linear}
    boneToModify {boneName=dyn_jacket_fr_l_01_Base}
    rotation=<rot 0 -20 0>
   }
   n30 = "Transform (Modify) Bone - Bone: dyn_jacket_fr_l_02"(ComponentPose=@n29, Alpha=@n28, Translation="0,-1,0", Rotation="0,-20,0", Scale="1,1,1") {
    rotationMode=BMM_Additive
    rotationSpace=BCS_BoneSpace
    scaleSpace=BCS_BoneSpace
    translationMode=BMM_Additive
    translationSpace=BCS_BoneSpace
    alphaBoolBlend {blendOption=Linear}
    boneToModify {boneName=dyn_jacket_fr_l_02}
    rotation {yaw=-20}
    translation {y=-1}
   }
   n31 = "Transform (Modify) Bone - Bone: dyn_jacket_fr_l_01_Base"(ComponentPose=@n30, Alpha=1, Translation="0,0,0", Rotation="-0.000,-8.359,0.000", Scale="1,1,1") {
    rotationMode=BMM_Additive
    rotationSpace=BCS_BoneSpace
    translationMode=BMM_Additive
    translationSpace=BCS_BoneSpace
    alphaScaleBiasClamp {bInterpResult=true}
    boneToModify {boneName=dyn_jacket_fr_l_01_Base}
    rotation=<rot 0 -8.359 0>
   }
   n32 = "Transform (Modify) Bone - Bone: dyn_jacket_bk_r_01_Base"(ComponentPose=@n31, Alpha=1, Translation="0,0,0", Rotation="8.916,-13.989,-0.778", Scale="1,1,1") {
    rotationMode=BMM_Additive
    rotationSpace=BCS_BoneSpace
    translationMode=BMM_Additive
    translationSpace=BCS_BoneSpace
    alphaScaleBiasClamp {bInterpResult=true}
    boneToModify {boneName=dyn_jacket_bk_r_01_Base}
    rotation=<rot 8.916 -13.989 -0.778>
   }
   Output Pose(Result=Component To Local(ComponentPose=@n32)) {layerGroup=None}
   n33 = Control Rig!(
    Source=Copy Pose From Mesh() {bUseAttachedParent=true}
    Alpha
    noRBAN_Mov:bool=Get NoRBAN_Mov() -> (NoRBAN_Mov:bool)
    noRigidBody:bool=NOT Boolean(A=Get bRigidBodyEnabled())
    isSkydiving:bool=Get bIsSkyDiving()
    isFrontEnd:bool=Get bIsInFrontEnd()
    isCrouch:bool=Get bIsCrouching()
    GoldenGloryEmote:double=Get GoldenGloryEmote() -> (GoldenGloryEmote:double)
    RigidBody:bool=Get bRigidBodyEnabled()
   ) -> (Pose) {
    alphaBoolBlend {blendOption=Linear}
    controlRigAssetReference {blueprintRigClass=&/F_MED_Photographer_Jacket_CtrlRig.F_MED_Photographer_Jacket_CtrlRig_C}
   }
   Save cached pose 'Photog'(Pose=@n33) {cacheName=Photog}
  ]}
 EventGraph {nodes [
   Event BlueprintUpdateAnimation()
   > n1 = Sequence()
   > Is Valid(InputObject=Get BodyMesh() -> (BodyMesh:Skeletal Mesh Component Object Reference))
   n2 = Get BodyMesh()
   > Set JacketFrontSideCorrection(JacketFrontSideCorrection:double=float * float(B=GetAlphaFromReferenceTransformDelta(SkeletalMeshComponent=@n2, BoneName=thigh_l, BaseName=pelvis, DeltaComponentType=QuaternionTwist, InputRange=(X=40,Y=106), TwistAxis="0,0,1")) {is=title K2Node_PromotableOperator}) -> (Output_Get:double)
   n3 = NOT Boolean(A=Get bIsPlayingEmote())
   n4 = GetAlphaFromReferenceTransformDelta(
    SkeletalMeshComponent=@n2
    BoneName=spine_05
    BaseName=spine_03
    DeltaComponentType=QuaternionTwist
    InputRange=(X=12,Y=18)
    TwistAxis="0,0,1"
   )
   > Set WeaponChange(WeaponChange:double=float * float(A="To Float (Boolean)"(InBool=@n3), B=Power!(Base:double=@n4, Exp:double=0.7) -> (ReturnValue:double) {is=title K2Node_CallFunction})) -> (Output_Get:double)
   n5 = Get bRigidBodyEnabled()
   > Set NoRBAN_Mov(NoRBAN_Mov:bool=AND Boolean(A=NOT Boolean(A=@n5), B=Get bIsMoving2D())) -> (Output_Get:bool)
   n6 = GetAnimInstance(self=GetOwnerSkeletalMeshComponent())
   n7 = OR Boolean(
    A=Montage_IsPlaying(self=@n6, Montage=/BRCosmetics/Animation/Game/MainPlayer/Emotes/GoldenGlory/CMF/Emote_GoldenGlory_CMF_M)
    B=Montage_IsPlaying(self=@n6, Montage=/BRCosmetics/Animation/Game/MainPlayer/Emotes/GoldenGlory/CMF/Emote_GoldenGlory_CMF_FrontEnd_M)
   )
   Set GoldenGloryEmote(execute=@n1.then_1, GoldenGloryEmote:double="To Float (Boolean)"(InBool=@n7)) -> (Output_Get:double)
   n8 = NOT Boolean(A=Get bRigidBodyEnabled())
   > Set GoldenGloryEmoteNoRban(GoldenGloryEmoteNoRban:double="To Float (Boolean)"(InBool=AND Boolean(A=@n7, B=@n8))) -> (Output_Get:double)
   Event OnCharacterCustomizationLoaded()
   n9 = CheckForBackbling()
   > n10 = Set Is Wearing Backbling(Is Wearing Backbling=OR Boolean(A=@n9, B=@n9.bIsBackblingLong))
   > n11 = Set Is Wearing Cape(Is Wearing Cape=@n9.bIsBackblingACape)
   > Set Backbling(Backbling:bool=OR Boolean(A=@n11, B=@n10)) -> (Output_Get:bool)
   > Set BodyMesh(BodyMesh=GetOwnerSkeletalMeshComponent())
   NOT Boolean(A=Get bIsMoving2D())
  ]}
]
variables [
 Backbling {type=boolean}
 BodyMesh {type=SkeletalMeshComponent}
 BootLaces_lenghtCorrection {type=number}
 DenimComponent Linear Acc Scale {type=Vector, default=<0.35 0.4 0.4>}
 DenimOverride World Gravity {type=Vector, default=<0 0 -500>}
 GoldenGloryEmote {type=number}
 GoldenGloryEmoteNoRban {type=number}
 HoodieComponent Linear Acc Scale {type=Vector, default=<0.4 0.3 0.3>}
 HoodieOverride World Gravity {type=Vector, default=<0 0 -980>}
 Is Wearing Backbling {type=boolean}
 Is Wearing Cape {type=boolean}
 JacketFrontSideCorrection {type=number}
 NoRBAN_Mov {type=boolean}
 parachuteClosed {type=boolean}
 Run_Correction_alpha {type=number}
 Sprint_Correction {type=number}
 WeaponChange {type=number}
]
