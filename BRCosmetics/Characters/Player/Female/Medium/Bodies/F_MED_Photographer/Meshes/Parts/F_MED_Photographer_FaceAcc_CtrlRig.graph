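# Control Rig for the female medium Photographer face accessory that procedurally aims back-of-head bones, likely hair, as the neck and head move.
# Half of neck_01's X rotation drives dyn_hairAimBACK, whose child target is reset before dyn_headBack_Base aims at it using a from-two-vectors rotation.
# Head pitch or bend relative to spine_05 then weights a target offset, and dyn_headBack_01_r alone aims at the shifted point.
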ControlRigBlueprint /BRCosmetics/Characters/Player/Female/Medium/Bodies/F_MED_Photographer/Meshes/Parts/F_MED_Photographer_FaceAcc_CtrlRig
graphs [
 RigVMFunctionLibrary {links [], nodes []}
 RigVMModel {nodes [
   n1 = &RigUnit_BeginExecution()
   RigUnit_GetTransform = &RigUnit_GetTransform(Item=(Type=Bone,Name="neck_01"), Space=LocalSpace)
   n2 = &RigVMFunction_MathQuaternionToEuler(RotationOrder=ZYX, Value=@RigUnit_GetTransform.Transform.Rotation)
   RigUnit_SetTransform = &RigUnit_SetTransform(
    Item=(Type=Bone,Name="dyn_hairAimBACK")
    Space=LocalSpace
    Value=(Rotation=(X=0.0284,Y=-0.7065,Z=-0.0284,W=0.7065),Translation=(X=0,Y=0,Z=0),Scale3D=(X=1,Y=1,Z=1))
    Weight=1
    bPropagateToChildren=false
    ExecutePin=@n1.ExecutePin
    Value.Rotation=&RigVMFunction_MathQuaternionFromEuler(
     Euler=(X=0,Y=0,Z=0)
     Euler.X=&RigVMFunction_MathFloatMul(B=0.5, A=@n2.Result.X) {template="Multiply::Execute(in A,in B,out Result)"}.Result
    ).Result
   )
   > &RigUnit_SetTransform(
    Item=(Type=Bone,Name="dyn_hairAimBACK_TGT")
    Space=LocalSpace
    Value=(Rotation=(X=0.0284,Y=-0.7065,Z=-0.0284,W=0.7065),Translation=(X=0,Y=-12,Z=0),Scale3D=(X=1,Y=1,Z=1))
    Weight=1
    bPropagateToChildren=false
   )
   n3 = &RigUnit_GetTransform(Item=(Type=Bone,Name="dyn_headBack_Base"))
   n4 = /Script/RigVM.RigVMFunction_MathQuaternionGetAxis(Axis=X, Quaternion=@n3.Transform.Rotation) {is=struct RigVMUnitNode, template="Axis::Execute(in Quaternion,in Axis,out Result)"}
   n5 = /Script/RigVM.RigVMFunction_MathVectorSub(
    A=/Script/RigVM.RigVMFunction_MathTransformTransformVector(
     Location=/Script/RigVM.RigVMFunction_MathVectorAdd(A=(X=0,Y=-1,Z=0), B=(X=0,Y=0,Z=0)) {is=struct RigVMUnitNode, template="Add::Execute(in A,in B,out Result)"}.Result
     Transform=&RigUnit_GetTransform(Item=(Type=Bone,Name="dyn_hairAimBACK_TGT")).Transform
    ) {is=struct RigVMUnitNode}.Result
    B=@n3.Transform.Translation
   ) {is=struct RigVMUnitNode, template="Subtract::Execute(in A,in B,out Result)"}
   > RigUnit_SetRotation = &RigUnit_SetRotation(
    Item=(Type=Bone,Name="dyn_headBack_Base")
    Space=GlobalSpace
    Weight=1
    Value=/Script/RigVM.RigVMFunction_MathQuaternionMul(
     A=/Script/RigVM.RigVMFunction_MathQuaternionFromTwoVectors(A=@n4.Result, B=@n5.Result) {is=struct RigVMUnitNode, template="FromTwoVectors::Execute(in A,in B,out Result)"}.Result
     B=@n3.Transform.Rotation
    ) {is=struct RigVMUnitNode, template="Multiply::Execute(in A,in B,out Result)"}.Result
   )
   n6 = &RigUnit_GetRelativeTransformForItem(Child=(Type=Bone,Name="head"), Parent=(Type=Bone,Name="spine_05"))
   > &RigUnit_ModifyTransforms(
    ItemToModify=((Item=(Type=Bone,Name="dyn_hairAimBACK_TGT"),Transform=(Rotation=(X=0,Y=0,Z=0,W=1),Translation=(X=0,Y=-20,Z=0),Scale3D=(X=1,Y=1,Z=1))))
    Weight=&RigVMFunction_MathFloatRemap(
     SourceMaximum=26
     SourceMinimum=-47
     TargetMinimum=0
     bClamp=true
     Value=&RigVMFunction_MathQuaternionToEuler(RotationOrder=ZYX, Value=@n6.RelativeTransform.Rotation).Result.Z
    ).Result
   )
   n7 = &RigUnit_GetTransform(Item=(Type=Bone,Name="dyn_headBack_01_r"))
   n8 = &RigVMFunction_MathQuaternionGetAxis(Axis=X, Quaternion=@n7.Transform.Rotation) {template="Axis::Execute(in Quaternion,in Axis,out Result)"}
   n9 = &RigVMFunction_MathVectorSub(
    A=&RigVMFunction_MathTransformTransformVector(
     Location=&RigVMFunction_MathVectorAdd(A=(X=0,Y=-2,Z=7), B=(X=0,Y=0,Z=0)) {template="Add::Execute(in A,in B,out Result)"}.Result
     Transform=&RigUnit_GetTransform(Item=(Type=Bone,Name="dyn_hairAimBACK_TGT")).Transform
    ).Result
    B=@n7.Transform.Translation
   ) {template="Subtract::Execute(in A,in B,out Result)"}
   > &RigUnit_SetRotation(
    Item=(Type=Bone,Name="dyn_headBack_01_r")
    Space=GlobalSpace
    Weight=1
    Value=&RigVMFunction_MathQuaternionMul(
     A=&RigVMFunction_MathQuaternionFromTwoVectors(A=@n8.Result, B=@n9.Result) {template="FromTwoVectors::Execute(in A,in B,out Result)"}.Result
     B=@n7.Transform.Rotation
    ) {template="Multiply::Execute(in A,in B,out Result)"}.Result
   )
  ]}
]
hierarchy [
 root {loc=<0 0 0>, rot=<0 0 0 1>}
 attach {parent=root, loc=<0 0 0>, rot=<0 0 0 1>}
 pelvis {parent=root, loc=<0 -0.5301 98.0571>, rot=<0.0127 0.707 -0.0127 -0.707>}
 spine_01 {parent=pelvis, loc=<6.1208 -0.9328 0>, rot=<0 0 0.034 -0.9994>}
 spine_02 {parent=spine_01, loc=<6.5996 -0.2095 0>, rot=<0 0 0.0238 -0.9997>}
 spine_03 {parent=spine_02, loc=<6.4914 -0.1237 0>, rot=<0 0 0.0149 -0.9999>}
 spine_04 {parent=spine_03, loc=<6.9598 0.0794 0>, rot=<0 0 -0.0132 -0.9999>}
 spine_05 {parent=spine_04, loc=<6.9987 0 0>, rot=<0 0 -0.0015 -1>}
 clavicle_l {parent=spine_05, loc=<5.4831 -0.1053 -3.2946>, rot=<0.0465 -0.7289 -0.009 -0.6829>}
 upperarm_l {parent=clavicle_l, loc=<9.8658 -1.2136 -0.4105>, rot=<0.0164 -0.3245 -0.0269 -0.9454>}
 lowerarm_l {parent=upperarm_l, loc=<26.465 -0.8345 0.259>, rot=<0.0036 0.0161 -0.2298 -0.9731>}
 hand_l {parent=lowerarm_l, loc=<22.7355 -0.3203 -0.6105>, rot=<-0.6784 -0.0223 0.0735 -0.7307>}
 index_metacarpal_l {parent=hand_l, loc=<1.6535 0.0183 -2.96>, rot=<0.0158 -0.0124 -0.0173 -0.9996>}
 index_01_l {parent=index_metacarpal_l, loc=<8.3008 0.0357 -0.0524>, rot=<-0.0044 0.0485 0.0924 -0.9945>}
 index_02_l {parent=index_01_l, loc=<3.3587 -0.3003 0.0452>, rot=<-0.002 0.0027 0.3225 -0.9466>}
 index_03_l {parent=index_02_l, loc=<2.5578 -0.123 -0.0252>, rot=<-0.002 0.0067 0.2558 -0.9667>}
 middle_metacarpal_l {parent=hand_l, loc=<2.1856 -0.1725 -0.5984>, rot=<0.006 0.0168 -0.0329 -0.9993>}
 middle_01_l {parent=middle_metacarpal_l, loc=<7.3275 -0.2447 0.0129>, rot=<-0.0208 0.0042 0.0858 -0.9961>}
 middle_02_l {parent=middle_01_l, loc=<4.0123 0.1401 0.1175>, rot=<-0.0218 0.0258 0.3512 -0.9357>}
 middle_03_l {parent=middle_02_l, loc=<2.6077 0.0534 0.1373>, rot=<0.0082 0.0188 0.1708 -0.9851>}
 pinky_metacarpal_l {parent=hand_l, loc=<1.7616 -0.9753 3.3814>, rot=<-0.0397 0.0814 -0.069 -0.9935>}
 pinky_01_l {parent=pinky_metacarpal_l, loc=<7.1709 -0.4831 -0.1094>, rot=<-0.0654 -0.0035 0.2242 -0.9723>}
 pinky_02_l {parent=pinky_01_l, loc=<2.8274 0.0893 -0.1043>, rot=<0.0037 0.0011 0.2854 -0.9584>}
 pinky_03_l {parent=pinky_02_l, loc=<2.2046 0.0308 -0.0078>, rot=<-0.0002 0.0098 0.1445 -0.9895>}
 ring_metacarpal_l {parent=hand_l, loc=<2.2074 -0.3625 1.3638>, rot=<-0.0284 0.0722 -0.0458 -0.9959>}
 ring_01_l {parent=ring_metacarpal_l, loc=<7.0295 -0.6197 -0.1446>, rot=<-0.0489 -0.0252 0.1892 -0.9804>}
 ring_02_l {parent=ring_01_l, loc=<3.6174 0.4647 -0.1027>, rot=<0 0 0.2806 -0.9598>}
 ring_03_l {parent=ring_02_l, loc=<2.4152 0.1141 0.0199>, rot=<0 0 0.1189 -0.9929>}
 thumb_01_l {parent=hand_l, loc=<1.3118 -0.3191 -2.459>, rot=<0.6546 -0.4042 0.0664 -0.6354>}
 thumb_02_l {parent=thumb_01_l, loc=<5.9336 -0.0448 0.0401>, rot=<-0.0037 0.0213 0.298 -0.9543>}
 thumb_03_l {parent=thumb_02_l, loc=<2.2223 0.3682 -0.0907>, rot=<0.0118 -0.013 0.0066 -0.9998>}
 weapon_l {parent=hand_l, loc=<-0.1974 -3.4111 0.3811>, rot=<0 0 0 -1>}
 lowerarm_twist_01_l {parent=lowerarm_l, loc=<18.316 0 0>, rot=<0 0 0 1>}
 lowerarm_twist_02_l {parent=lowerarm_l, loc=<10.6707 0 0>, rot=<0 0 0 1>}
 dyn_cuffSide_Base_l {parent=lowerarm_twist_02_l, loc=<3.3112 -3.6835 0.1797>, rot=<0.9719 0.2247 0.056 -0.0421>}
 dyn_cuffSide_l {parent=dyn_cuffSide_Base_l, loc=<0 0 0>, rot=<0 0 0 1>}
 dyn_cuffTop_Base_l {parent=lowerarm_twist_02_l, loc=<3.3496 -0.2624 3.8114>, rot=<-0.6963 -0.1855 0.183 -0.6688>}
 dyn_cuffTop_l {parent=dyn_cuffTop_Base_l, loc=<0 0 0>, rot=<0 0 0 -1>}
 upperarm_twist_01_l {parent=upperarm_l, loc=<7.2135 -0.1134 0>, rot=<0 0 0 1>}
 upperarm_twist_02_l {parent=upperarm_l, loc=<19.7998 -0.1134 0>, rot=<0 0 0 1>}
 dyn_collarB_Base_l {parent=clavicle_l, loc=<1.5664 1.0211 7.0059>, rot=<-0.1366 -0.7591 -0.6359 -0.0274>}
 dyn_collarB_l {parent=dyn_collarB_Base_l, loc=<0 0 0>, rot=<0 0 0 1>}
 dyn_collarA_Base_l {parent=clavicle_l, loc=<1.4409 6.9726 2.7188>, rot=<-0.0888 -0.9063 -0.4131 0.0132>}
 dyn_collarA_l {parent=dyn_collarA_Base_l, loc=<0 0 0>, rot=<0 0 0 1>}
 clavicle_r {parent=spine_05, loc=<5.4831 -0.1053 3.2946>, rot=<0.7289 0.0465 -0.6829 0.009>}
 upperarm_r {parent=clavicle_r, loc=<-9.8658 1.2136 0.4104>, rot=<0.0164 -0.3245 -0.0269 -0.9454>}
 lowerarm_r {parent=upperarm_r, loc=<-26.4651 0.8345 -0.259>, rot=<0.0036 0.0161 -0.2298 -0.9731>}
 hand_r {parent=lowerarm_r, loc=<-22.7356 0.3203 0.6106>, rot=<-0.6784 -0.0223 0.0735 -0.7307>}
 index_metacarpal_r {parent=hand_r, loc=<-1.6534 -0.0181 2.9601>, rot=<0.0158 -0.0124 -0.0173 -0.9996>}
 index_01_r {parent=index_metacarpal_r, loc=<-8.3008 -0.0357 0.0524>, rot=<-0.0044 0.0485 0.0924 -0.9945>}
 index_02_r {parent=index_01_r, loc=<-3.3589 0.3003 -0.0453>, rot=<-0.002 0.0027 0.3225 -0.9466>}
 index_03_r {parent=index_02_r, loc=<-2.5578 0.123 0.0252>, rot=<-0.002 0.0067 0.2558 -0.9667>}
 middle_metacarpal_r {parent=hand_r, loc=<-2.1856 0.1726 0.5984>, rot=<0.006 0.0168 -0.0329 -0.9993>}
 middle_01_r {parent=middle_metacarpal_r, loc=<-7.3273 0.2447 -0.0129>, rot=<-0.0208 0.0042 0.0858 -0.9961>}
 middle_02_r {parent=middle_01_r, loc=<-4.0124 -0.14 -0.1175>, rot=<-0.0218 0.0258 0.3512 -0.9357>}
 middle_03_r {parent=middle_02_r, loc=<-2.6077 -0.0533 -0.1373>, rot=<0.0082 0.0188 0.1708 -0.9851>}
 pinky_metacarpal_r {parent=hand_r, loc=<-1.7617 0.9753 -3.3814>, rot=<-0.0397 0.0814 -0.069 -0.9935>}
 pinky_01_r {parent=pinky_metacarpal_r, loc=<-7.1707 0.4832 0.1094>, rot=<-0.0654 -0.0035 0.2242 -0.9723>}
 pinky_02_r {parent=pinky_01_r, loc=<-2.8274 -0.0892 0.1043>, rot=<0.0037 0.0011 0.2854 -0.9584>}
 pinky_03_r {parent=pinky_02_r, loc=<-2.2045 -0.031 0.0078>, rot=<-0.0002 0.0098 0.1445 -0.9895>}
 ring_metacarpal_r {parent=hand_r, loc=<-2.2073 0.3627 -1.3638>, rot=<-0.0284 0.0722 -0.0458 -0.9959>}
 ring_01_r {parent=ring_metacarpal_r, loc=<-7.0297 0.6197 0.1446>, rot=<-0.0489 -0.0252 0.1892 -0.9804>}
 ring_02_r {parent=ring_01_r, loc=<-3.6172 -0.4647 0.1027>, rot=<0 0 0.2806 -0.9598>}
 ring_03_r {parent=ring_02_r, loc=<-2.4152 -0.1141 -0.0199>, rot=<0 0 0.1189 -0.9929>}
 thumb_01_r {parent=hand_r, loc=<-1.3119 0.3192 2.459>, rot=<0.6546 -0.4042 0.0664 -0.6354>}
 thumb_02_r {parent=thumb_01_r, loc=<-5.9336 0.0448 -0.0401>, rot=<-0.0037 0.0213 0.298 -0.9543>}
 thumb_03_r {parent=thumb_02_r, loc=<-2.2223 -0.3682 0.0907>, rot=<0.0118 -0.013 0.0066 -0.9998>}
 weapon_r {parent=hand_r, loc=<0.1975 3.4111 -0.3811>, rot=<0 0 0 -1>}
 lowerarm_twist_01_r {parent=lowerarm_r, loc=<-18.3161 0 0>, rot=<0 0 0 1>}
 lowerarm_twist_02_r {parent=lowerarm_r, loc=<-10.6707 0 0>, rot=<0 0 0 1>}
 dyn_cuffSide_Base_r {parent=lowerarm_twist_02_r, loc=<-3.3109 3.6834 -0.18>, rot=<0.9719 0.2247 0.056 -0.0421>}
 dyn_cuffSide_r {parent=dyn_cuffSide_Base_r, loc=<0 0 0>, rot=<0 0 0 1>}
 dyn_cuffTop_Base_r {parent=lowerarm_twist_02_r, loc=<-3.3495 0.2623 -3.8114>, rot=<-0.6963 -0.1855 0.183 -0.6688>}
 dyn_cuffTop_r {parent=dyn_cuffTop_Base_r, loc=<0 0 0>, rot=<0 0 0 1>}
 upperarm_twist_01_r {parent=upperarm_r, loc=<-7.2135 0.1134 0>, rot=<0 0 0 1>}
 upperarm_twist_02_r {parent=upperarm_r, loc=<-19.7998 0.1134 0>, rot=<0 0 0 1>}
 dyn_collarA_Base_r {parent=clavicle_r, loc=<-1.4409 -6.9726 -2.719>, rot=<-0.0888 -0.9063 -0.4131 0.0132>}
 dyn_collarA_r {parent=dyn_collarA_Base_r, loc=<0 0 0>, rot=<0 0 0 1>}
 dyn_collarB_Base_r {parent=clavicle_r, loc=<-1.5663 -1.0211 -7.0061>, rot=<-0.1366 -0.7591 -0.6359 -0.0274>}
 dyn_collarB_r {parent=dyn_collarB_Base_r, loc=<0 0 0>, rot=<0 0 0 1>}
 neck_01 {parent=spine_05, loc=<11.3138 0.1529 0>, rot=<0 0 -0.2003 -0.9797>}
 neck_02 {parent=neck_01, loc=<4.5627 0 0>, rot=<0 0 0.0504 -0.9987>}
 head {parent=neck_02, loc=<5.3914 0 0>, rot=<0 0 0.115 -0.9934>}
 dyn_headBack_Base {parent=head, loc=<6.9539 -5.7482 -0.0677>, rot=<0.1807 -0.9835 0 0>}
 dyn_headBack_01_c {parent=dyn_headBack_Base, loc=<5.269 -0.4697 -0.1062>, rot=<0 0 -0.0203 -0.9998>}
 dyn_headBack_02_c {parent=dyn_headBack_01_c, loc=<3.9528 0 0>, rot=<0 0 -0.0156 -0.9999>}
 dyn_headBack_03_c {parent=dyn_headBack_02_c, loc=<4.6852 0 0>, rot=<0 0 0 -1>}
 dyn_headBack_01_l {parent=dyn_headBack_Base, loc=<4.8669 1.4907 5.4464>, rot=<0.4407 0.2004 0.0062 -0.875>}
 dyn_headBack_02_l {parent=dyn_headBack_01_l, loc=<4.1345 0 0>, rot=<-0.0393 -0.0144 0.0264 -0.9988>}
 dyn_headBack_03_l {parent=dyn_headBack_02_l, loc=<4.5827 0 0>, rot=<0 0 0 -1>}
 dyn_headBack_01_r {parent=dyn_headBack_Base, loc=<4.6666 2.2033 -5.0424>, rot=<-0.1882 -0.0662 -0.0164 -0.9798>}
 dyn_headBack_02_r {parent=dyn_headBack_01_r, loc=<4.1345 0 0>, rot=<-0.1285 -0.0538 0.0742 -0.9875>}
 dyn_headBack_03_r {parent=dyn_headBack_02_r, loc=<4.5827 0 0>, rot=<0 0 0 -1>}
 dyn_hatWing_01_r {parent=head, loc=<12.3436 6.8143 5.4574>, rot=<0.8396 -0.3869 0.1625 0.345>}
 dyn_hatWing_02_r {parent=dyn_hatWing_01_r, loc=<6.2474 0 0>, rot=<-0.0073 0.0075 0.0727 -0.9973>}
 dyn_hatWing_03_r {parent=dyn_hatWing_02_r, loc=<5.2555 0 0>, rot=<0 0 0 -1>}
 dyn_hatWing_01_l {parent=head, loc=<12.3436 6.8143 -5.4574>, rot=<0.8573 -0.3722 -0.1444 -0.3251>}
 dyn_hatWing_02_l {parent=dyn_hatWing_01_l, loc=<5.7908 0 0>, rot=<-0.033 0.0345 0.0196 -0.9987>}
 dyn_hatWing_03_l {parent=dyn_hatWing_02_l, loc=<5.7083 0 0>, rot=<0 0 0 -1>}
 dyn_hairFrontBounce_01_Base {parent=head, loc=<2.7792 4.7152 7.7078>, rot=<-0.001 0.0201 0.9989 0.0413>}
 dyn_hairFrontBounce_01 {parent=dyn_hairFrontBounce_01_Base, loc=<0 0 0>, rot=<0 0 0 1>}
 dyn_hairFrontBounce_02 {parent=dyn_hairFrontBounce_01, loc=<5.6405 0 0>, rot=<-0.3719 -0.3508 0.5305 -0.6761>}
 dyn_headFront_01_Base_r {parent=head, loc=<2.7793 2.2241 6.1332>, rot=<0.0631 -0.884 0.4623 0.0292>}
 dyn_headFront_01_r {parent=dyn_headFront_01_Base_r, loc=<0 0 0>, rot=<0 0 0 1>}
 dyn_headFront_02_r {parent=dyn_headFront_01_r, loc=<5.4478 0 0>, rot=<-0.5562 -0.2621 0.0196 -0.7884>}
 dyn_headFront_03_r {parent=dyn_headFront_02_r, loc=<4.6029 0 0>, rot=<-0.0105 0.1028 0.6868 -0.7194>}
 dyn_headFront_01_Base_l {parent=head, loc=<3.2526 3.9383 -7.1179>, rot=<-0.1444 0.6533 0.7368 -0.0971>}
 dyn_headFront_01_l {parent=dyn_headFront_01_Base_l, loc=<0 0 0>, rot=<0 0 0 1>}
 dyn_headFront_02_l {parent=dyn_headFront_01_l, loc=<3.9425 0 0>, rot=<0.0227 0.1633 -0.1007 -0.9812>}
 dyn_headFront_03_l {parent=dyn_headFront_02_l, loc=<4.3209 0 0>, rot=<0.1723 0.2219 0.1738 -0.9439>}
 dyn_headFront_04_l {parent=dyn_headFront_03_l, loc=<2.9219 0 0>, rot=<0.09 -0.1263 0.6771 -0.7194>}
 dyn_head_01_Base_l {parent=head, loc=<3.1271 -1.7629 -8.368>, rot=<0.0456 -0.7633 -0.6422 0.0531>}
 dyn_head_01_l {parent=dyn_head_01_Base_l, loc=<0 0 0>, rot=<0 0 0 -1>}
 dyn_head_02_l {parent=dyn_head_01_l, loc=<5.3391 0 0>, rot=<0.0863 0.0823 0.1845 -0.9756>}
 dyn_head_03_l {parent=dyn_head_02_l, loc=<4.1111 0 0>, rot=<0 0 0 1>}
 dyn_head_01_Base_r {parent=head, loc=<2.3393 -2.5791 6.7725>, rot=<0.066 -0.704 0.7045 0.0602>}
 dyn_head_01_r {parent=dyn_head_01_Base_r, loc=<0 0 0>, rot=<0 0 0 -1>}
 dyn_head_02_r {parent=dyn_head_01_r, loc=<5.3391 0 0>, rot=<0.0786 0.0273 0.1549 -0.9844>}
 dyn_head_03_r {parent=dyn_head_02_r, loc=<4.1111 0 0>, rot=<0 0 0 -1>}
 dyn_camMotion_01_Base {parent=spine_05, loc=<0.6423 9.5213 -0.0003>, rot=<-0.2217 -0.9751 0.0013 0.0041>}
 dyn_camMotion_01 {parent=dyn_camMotion_01_Base, loc=<0 0 0>, rot=<0 0 0 -1>}
 dyn_camMotion_02 {parent=dyn_camMotion_01, loc=<6.0247 0 0>, rot=<-0.0003 -0.0008 0.1432 -0.9897>}
 dyn_camMotion_03 {parent=dyn_camMotion_02, loc=<9.8554 0 0>, rot=<0.0002 0.0018 0.1197 -0.9928>}
 dyn_camMotion_04 {parent=dyn_camMotion_03, loc=<8.1282 0 0>, rot=<0 0.7095 0 -0.7047>}
 dyn_camStrap_03_r {parent=dyn_camMotion_03, loc=<-0.7089 -2.1021 -3.0205>, rot=<0 0.7095 0 -0.7047>}
 dyn_camStrap_03_l {parent=dyn_camMotion_03, loc=<-0.6018 -2.2079 3.0062>, rot=<0 0.7095 0 -0.7047>}
 dyn_camStrap_01_Base_r {parent=dyn_camMotion_01, loc=<6.1253 0.1751 -3.8278>, rot=<0.0098 0.0246 0.1469 -0.9888>}
 dyn_camStrap_01_r {parent=dyn_camStrap_01_Base_r, loc=<0 0 0>, rot=<0 0 0 1>}
 dyn_camStrap_02_r {parent=dyn_camStrap_01_r, loc=<4.0891 0 0>, rot=<0.0008 0.0324 0.2052 -0.9782>}
 dyn_camStrap_01_Base_l {parent=dyn_camMotion_01, loc=<6.1858 0.2936 3.9718>, rot=<-0.0137 -0.0404 0.1976 -0.9793>}
 dyn_camStrap_01_l {parent=dyn_camStrap_01_Base_l, loc=<0 0 0>, rot=<0 0 0 -1>}
 dyn_camStrap_02_l {parent=dyn_camStrap_01_l, loc=<4.0234 0 0>, rot=<0.0001 -0.0213 0.1348 -0.9906>}
 dyn_collarC_Base_r {parent=spine_05, loc=<13.3481 -3.7102 3.914>, rot=<0.3758 -0.8534 0.1992 0.3013>}
 dyn_collarC_r {parent=dyn_collarC_Base_r, loc=<0 0 0>, rot=<0 0 0 1>}
 dyn_collarBack_Base {parent=spine_05, loc=<13.4111 -4.8047 0>, rot=<0.3456 -0.9382 -0.0164 0.006>}
 dyn_collarBack {parent=dyn_collarBack_Base, loc=<0 0 0>, rot=<0 0 0 1>}
 dyn_collarC_Base_l {parent=spine_05, loc=<13.3482 -3.7102 -3.914>, rot=<0.8534 0.3758 0.3013 -0.1992>}
 dyn_collarC_l {parent=dyn_collarC_Base_l, loc=<0 0 0>, rot=<0 0 0 1>}
 dyn_hairAimBACK {parent=spine_05, loc=<-0.286 3.558 0>, rot=<0 0 0 1>}
 dyn_hairAimBACK_TGT {parent=dyn_hairAimBACK, loc=<0.286 -3.558 0>, rot=<0 0 0 1>}
 dyn_jacket_bk_01 {parent=spine_04, loc=<-2.4956 -8.767 0>, rot=<-0.0936 -0.9956 0 0>}
 dyn_jacket_bk_02 {parent=dyn_jacket_bk_01, loc=<7.5517 0 0>, rot=<0 0 0.2756 -0.9613>}
 dyn_jacket_bk_03 {parent=dyn_jacket_bk_02, loc=<9.3259 0 0>, rot=<0 0 0 1>}
 dyn_jacket_bk_r_01_Base {parent=spine_04, loc=<-3.0432 -6.5468 8.9164>, rot=<-0.0443 -0.9695 0.2404 -0.0159>}
 dyn_jacket_bk_r_01 {parent=dyn_jacket_bk_r_01_Base, loc=<0 0 0>, rot=<0 0 0 1>}
 dyn_jacket_bk_r_02 {parent=dyn_jacket_bk_r_01, loc=<8.5515 0 0>, rot=<-0.0068 0.0314 0.2479 -0.9683>}
 dyn_jacket_bk_r_03 {parent=dyn_jacket_bk_r_02, loc=<7.1853 0 0>, rot=<-0.0349 0 0 -0.9994>}
 dyn_jacket_bk_l_01_Base {parent=spine_04, loc=<-3.0429 -6.5467 -8.9164>, rot=<0.9695 -0.0443 -0.0159 -0.2404>}
 dyn_jacket_bk_l_01 {parent=dyn_jacket_bk_l_01_Base, loc=<0 0 0>, rot=<0 0 0 1>}
 dyn_jacket_bk_l_02 {parent=dyn_jacket_bk_l_01, loc=<-8.5521 0.0001 0>, rot=<-0.0406 0.0401 0.2467 -0.9674>}
 dyn_jacket_bk_l_03 {parent=dyn_jacket_bk_l_02, loc=<-7.1849 0 0>, rot=<0 0 0 -1>}
 dyn_jacket_r_01_Base {parent=spine_04, loc=<-4.6047 0.0697 11.7338>, rot=<0.0744 -0.7556 0.6508 0.0098>}
 dyn_jacket_r_01 {parent=dyn_jacket_r_01_Base, loc=<0 0 0>, rot=<0 0 0 1>}
 dyn_jacket_r_02 {parent=dyn_jacket_r_01, loc=<8.5184 0 0>, rot=<-0.036 -0.0074 0.09 -0.9953>}
 dyn_jacket_r_03 {parent=dyn_jacket_r_02, loc=<7.4561 0 0>, rot=<0 0 0 -1>}
 dyn_jacket_l_01_Base {parent=spine_04, loc=<-4.6042 0.0697 -11.7338>, rot=<0.7556 0.0744 0.0098 -0.6508>}
 dyn_jacket_l_01 {parent=dyn_jacket_l_01_Base, loc=<0 0 0>, rot=<0 0 0 -1>}
 dyn_jacket_l_02 {parent=dyn_jacket_l_01, loc=<-8.5185 0 0>, rot=<-0.036 -0.0074 0.0899 -0.9953>}
 dyn_jacket_l_03 {parent=dyn_jacket_l_02, loc=<-7.4568 0 0>, rot=<0 0 0 -1>}
 dyn_jacket_fr_r_01_Base {parent=spine_04, loc=<-5.2872 8.9229 9.349>, rot=<0.0621 -0.4103 0.908 -0.0582>}
 dyn_jacket_fr_r_01 {parent=dyn_jacket_fr_r_01_Base, loc=<0 0 0>, rot=<0 0 0 1>}
 dyn_jacket_fr_r_02 {parent=dyn_jacket_fr_r_01, loc=<10.1002 0 0>, rot=<0.0619 0.0239 0.1082 -0.9919>}
 dyn_jacket_fr_r_03 {parent=dyn_jacket_fr_r_02, loc=<7.4598 0 0>, rot=<-0.0872 0 0 -0.9962>}
 dyn_jacket_fr_l_01_Base {parent=spine_04, loc=<-5.2874 8.9229 -9.349>, rot=<0.4103 0.0621 -0.0582 -0.908>}
 dyn_jacket_fr_l_01 {parent=dyn_jacket_fr_l_01_Base, loc=<0 0 0>, rot=<0 0 0 1>}
 dyn_jacket_fr_l_02 {parent=dyn_jacket_fr_l_01, loc=<-10.1003 0 0>, rot=<0.0272 0.0277 0.1073 -0.9935>}
 dyn_jacket_fr_l_03 {parent=dyn_jacket_fr_l_02, loc=<-7.4593 0 0>, rot=<-0.0523 0 0 -0.9986>}
 thigh_l {parent=pelvis, loc=<-1.0855 -0.2109 -8.7053>, rot=<0.0009 0.0355 0.0155 -0.9993>}
 calf_l {parent=thigh_l, loc=<-42.5284 -0.0019 -0.7297>, rot=<0.0006 -0.0161 -0.0391 -0.9991>}
 calf_twist_01_l {parent=calf_l, loc=<-35.8483 -1.9197 -0.5794>, rot=<0.0237 0.0095 -0.0035 -0.9997>}
 dyn_lacesBase_l {parent=calf_twist_01_l, loc=<4.3884 3.91 -1.0171>, rot=<0 0 0 -1>}
 dyn_laceInner_l {parent=dyn_lacesBase_l, loc=<-0.2393 -0.0541 0.3103>, rot=<0.1496 -0.2186 0.9641 0.0206>}
 dyn_laceOuter_l {parent=dyn_lacesBase_l, loc=<-0.5067 -0.1259 -0.5629>, rot=<-0.0849 0.0849 0.9923 0.0311>}
 calf_twist_02_l {parent=calf_l, loc=<-19.1546 -1.7943 -0.2638>, rot=<0.0237 0.0095 -0.0035 -0.9997>}
 foot_l {parent=calf_l, loc=<-46.7401 -2.5014 -0.592>, rot=<0.0427 -0.0238 0.0353 -0.9982>}
 ball_l {parent=foot_l, loc=<-3.5658 10.8066 -0.1544>, rot=<0.0001 -0.0001 -0.7063 -0.7079>}
 thigh_twist_01_l {parent=thigh_l, loc=<-27.952 0 0>, rot=<0.0013 0 0 -1>}
 dyn_hoodieOverleg_01_Base_l {parent=thigh_l, loc=<2.6785 4.1363 -4.0115>, rot=<0.1842 0.3502 0.9106 0.1196>}
 dyn_hoodieOverleg_01_l {parent=dyn_hoodieOverleg_01_Base_l, loc=<0 0 0>, rot=<0 0 0 1>}
 dyn_hoodieOverleg_02_l {parent=dyn_hoodieOverleg_01_l, loc=<6.0424 0 0>, rot=<0 0 0 1>}
 thigh_r {parent=pelvis, loc=<-1.0853 -0.2109 8.7053>, rot=<0.0355 -0.0009 0.9993 0.0155>}
 calf_r {parent=thigh_r, loc=<42.5286 0.0019 0.7297>, rot=<0.0006 -0.0161 -0.0391 -0.9991>}
 calf_twist_01_r {parent=calf_r, loc=<35.8485 1.9197 0.5794>, rot=<0.0237 0.0095 -0.0035 -0.9997>}
 dyn_lacesBase_r {parent=calf_twist_01_r, loc=<-4.3901 -3.8497 1.0664>, rot=<0 0 0 1>}
 dyn_laceOuter_r {parent=dyn_lacesBase_r, loc=<0.5082 0.0656 0.5137>, rot=<0.0849 0.0849 0.0311 -0.9923>}
 dyn_laceInner_r {parent=dyn_lacesBase_r, loc=<0.4165 0.055 -0.588>, rot=<-0.224 -0.138 0.0226 -0.9645>}
 calf_twist_02_r {parent=calf_r, loc=<19.1546 1.7943 0.2638>, rot=<0.0237 0.0095 -0.0035 -0.9997>}
 foot_r {parent=calf_r, loc=<46.7403 2.5014 0.592>, rot=<0.0427 -0.0238 0.0353 -0.9982>}
 ball_r {parent=foot_r, loc=<3.5658 -10.8066 0.1544>, rot=<0.0001 -0.0001 -0.7063 -0.7079>}
 thigh_twist_01_r {parent=thigh_r, loc=<27.952 0 0>, rot=<0.0013 0 0 -1>}
 dyn_pouch_Base_r {parent=thigh_twist_01_r, loc=<-20.217 -2.019 7.0362>, rot=<0.5916 0.0066 0.0092 -0.8062>}
 dyn_pouch_r {parent=dyn_pouch_Base_r, loc=<0 0 0>, rot=<0 0 0 -1>}
 dyn_tassle {parent=dyn_pouch_r, loc=<17.0958 -2.0319 0.1739>, rot=<0.0055 -0.0038 -0.0342 -0.9994>}
 dyn_hoodieKnot {parent=pelvis, loc=<2.2378 9.0787 0.3498>, rot=<0 0 -0.629 -0.7774>}
 dyn_hoodieFront_01_Base_l {parent=dyn_hoodieKnot, loc=<-0.0015 2.0872 -2.2285>, rot=<0.0361 -0.0987 -0.7649 -0.6356>}
 dyn_hoodieFront_01_l {parent=dyn_hoodieFront_01_Base_l, loc=<0 0 0>, rot=<0 0 0 1>}
 dyn_hoodieFront_02_l {parent=dyn_hoodieFront_01_l, loc=<6.8983 0 0>, rot=<-0.0132 -0.0539 0.0228 -0.9982>}
 dyn_hoodieFront_03_l {parent=dyn_hoodieFront_02_l, loc=<6.9026 0 0>, rot=<-0.0821 0.0983 0.0078 -0.9917>}
 dyn_hoodieFront_01_Base_r {parent=dyn_hoodieKnot, loc=<-0.2264 2.722 2.748>, rot=<-0.0334 0.0736 -0.7652 -0.6387>}
 dyn_hoodieFront_01_r {parent=dyn_hoodieFront_01_Base_r, loc=<0 0 0>, rot=<0 0 0 1>}
 dyn_hoodieFront_02_r {parent=dyn_hoodieFront_01_r, loc=<6.4114 0 0>, rot=<0.1099 -0.0118 0.0505 -0.9926>}
 dyn_hoodieFront_03_r {parent=dyn_hoodieFront_02_r, loc=<6.3628 0 0>, rot=<0.0021 0.1007 -0.039 -0.9942>}
 dyn_hoodieOverleg_01_Base_r {parent=pelvis, loc=<3.3644 0.9373 13.1881>, rot=<-0.0543 -0.553 0.8137 0.1708>}
 dyn_hoodieOverleg_01_r {parent=dyn_hoodieOverleg_01_Base_r, loc=<0 0 0>, rot=<0 0 0 -1>}
 dyn_hoodieOverleg_02_r {parent=dyn_hoodieOverleg_01_r, loc=<3.8869 0 0>, rot=<0 0 0.061 -0.9981>}
 dyn_hoodie_01_Base_l {parent=pelvis, loc=<-7.1664 -8.4608 -11.5241>, rot=<-0.0453 -0.9678 -0.2421 -0.0513>}
 dyn_hoodie_01_l {parent=dyn_hoodie_01_Base_l, loc=<0 0 0>, rot=<0 0 0 -1>}
 dyn_hoodie_02_l {parent=dyn_hoodie_01_l, loc=<7.0159 0 0>, rot=<0.0184 -0.0817 0.1062 -0.9908>}
 dyn_hoodie_03_l {parent=dyn_hoodie_02_l, loc=<9.5717 0 0>, rot=<-0.066 -0.028 0.0188 -0.9973>}
 dyn_hoodie_01_c_Base {parent=pelvis, loc=<-9.1694 -10.4561 0.028>, rot=<-0.0039 -1 -0.0001 0.0039>}
 dyn_hoodie_01_c {parent=dyn_hoodie_01_c_Base, loc=<0 0 0>, rot=<0 0 0 -1>}
 dyn_hoodie_02_c {parent=dyn_hoodie_01_c, loc=<7.1432 0 0>, rot=<0.0002 -0.0021 0.063 -0.998>}
 dyn_hoodie_03_c {parent=dyn_hoodie_02_c, loc=<8.629 0 0>, rot=<-0.0016 0.0107 -0.0006 -0.9999>}
 dyn_hoodie_01_Base_r {parent=pelvis, loc=<-7.9154 -10.1619 9.1763>, rot=<-0.0067 -0.9859 0.155 0.0621>}
 dyn_hoodie_01_r {parent=dyn_hoodie_01_Base_r, loc=<0 0 0>, rot=<0 0 0 1>}
 dyn_hoodie_02_r {parent=dyn_hoodie_01_r, loc=<7.1926 0 0>, rot=<0.0657 0.087 0.0238 -0.9937>}
 dyn_hoodie_03_r {parent=dyn_hoodie_02_r, loc=<8.2666 -0.0111 -0.0631>, rot=<-0.0021 0.0231 -0.0113 -0.9997>}
 ik_foot_root {parent=root, loc=<0 0 0>, rot=<0 0 0 1>}
 ik_foot_l {parent=ik_foot_root, loc=<14.8262 -7.3545 8.296>, rot=<-0.0357 -0.7045 -0.0284 0.7082>}
 ik_foot_r {parent=ik_foot_root, loc=<-14.8262 -7.3546 8.2958>, rot=<-0.7082 -0.0284 0.7045 -0.0357>}
 ik_hand_root {parent=root, loc=<0 0 0>, rot=<0 0 0 1>}
 ik_hand_gun {parent=ik_hand_root, loc=<-48.1744 3.3546 104.4381>, rot=<-0.6276 -0.1309 -0.4183 0.6434>}
 ik_hand_l {parent=ik_hand_gun, loc=<59.3272 67.6997 34.3524>, rot=<-0.6158 -0.7027 -0.3565 0>}
 ik_hand_r {parent=ik_hand_gun, loc=<0 0 0>, rot=<0 0 0 -1>}
]
