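# Physics asset for the female medium Photographer's face accessory mesh, setting up secondary motion for dynamic hair and hat-wing bones.
# Capsules on the upper spine, right clavicle, neck and head form the body shapes, joined by angularly free constraints.
# Hat-wing, front, side and back chains branch from the head, mostly through base bodies, with narrow swing and twist limits.
# Extra free-rotation constraints cross-link back-chain tips and the right side tip, and the center back chain locks twist.
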PhysicsAsset /BRCosmetics/Characters/Player/Female/Medium/Bodies/F_MED_Photographer/Meshes/Parts/F_MED_Photographer_FaceAcc_Physics_FE
bodies {
 clavicle_r [{shapeType=Capsule, length=7.3396, radius=7.3723, center=<-5.228 0.9491 0.3525>, rotation=<rot 88.9241 141.0166 144.8575>}]
 dyn_hairFrontBounce_01 [{shapeType=Capsule, length=2, radius=1.75, shapeName=dyn_hairFrontBounce_01_capsule, center=<2.6295 0 0>, rotation=<rot 90 0.0395 0.0394>}]
 dyn_hairFrontBounce_01_Base [{shapeType=Capsule, length=0.505, radius=0.505, shapeName=dyn_hairFrontBounce_01_Base_capsule, rotation=<rot 0 -90 0>}]
 dyn_hairFrontBounce_02 [{
   shapeType=Capsule
   length=1.25
   radius=1.25
   shapeName=dyn_hairFrontBounce_02_capsule
   center=<1.4096 0.8185 0.1442>
   rotation=<rot -9.8472 -79.848 88.246>
  }]
 dyn_hatWing_01_l [{shapeType=Capsule, length=2, radius=2, shapeName=dyn_hatWing_01_l_capsule}]
 dyn_hatWing_01_r [{shapeType=Capsule, length=2, radius=2, shapeName=dyn_hatWing_01_r_capsule}]
 dyn_hatWing_02_l [{shapeType=Capsule, length=2, radius=2, shapeName=dyn_hatWing_02_l_capsule}]
 dyn_hatWing_02_r [{shapeType=Capsule, length=2, radius=2, shapeName=dyn_hatWing_02_r_capsule}]
 dyn_headBack_02_c [{shapeType=Capsule, length=1.75, radius=2, shapeName=dyn_headBack_02_c_capsule, center=<0.7895 -0.4903 0>, rotation=<rot 90 180 -175.0003>}]
 dyn_headBack_02_l [{
   shapeType=Capsule
   length=1.25
   radius=1.5
   shapeName=dyn_headBack_02_l_capsule
   center=<1.003 -0.5288 -0.0001>
   rotation=<rot 90 -179.9736 -179.9737>
  }]
 dyn_headBack_02_r [{shapeType=Capsule, length=1.75, radius=1.75, shapeName=dyn_headBack_02_r_capsule, center=<2.1145 0 0>, rotation=<rot 90 0 0>}]
 dyn_headBack_03_c [{shapeType=Capsule, length=4.2818, radius=3.9503, shapeName=dyn_headBack_03_c_capsule, center=<0.8979 -1.5139 0>}]
 dyn_headBack_03_l [{shapeType=Capsule, length=4.4164, radius=3.8, shapeName=dyn_headBack_03_l_capsule, center=<-0.3572 -0.0011 -1.5826>}]
 dyn_headBack_03_r [{
   shapeType=Capsule
   length=5.0545
   radius=3.4183
   shapeName=dyn_headBack_03_r_capsule
   center=<0.3149 -1.0318 0.6101>
   rotation=<rot -2.4519 -90 11.8998>
  }]
 dyn_headBack_Base [{shapeType=Sphere, radius=2, shapeName=dyn_headBack_Base_sphere, center=<2.8492 -2.229 -1.2497>}]
 dyn_headFront_01_Base_l [{shapeType=Capsule, length=0.505, radius=0.505, shapeName=dyn_headFront_01_Base_l_capsule, rotation=<rot 0 -90 0>}]
 dyn_headFront_01_Base_r [{shapeType=Capsule, length=0.505, radius=0.505, shapeName=dyn_headFront_01_Base_r_capsule, rotation=<rot 0 -90 0>}]
 dyn_headFront_01_l [{shapeType=Capsule, length=1.25, radius=2, shapeName=dyn_headFront_01_l_capsule, center=<2.0487 0 0>, rotation=<rot -90 -0.0198 0.0198>}]
 dyn_headFront_01_r [{
   shapeType=Capsule
   length=2
   radius=2
   shapeName=dyn_headFront_01_r_capsule
   center=<2.3042 -1.3625 -0.7411>
   rotation=<rot 49.9993 89.9998 89.9999>
  }]
 dyn_headFront_02_l [{
   shapeType=Capsule
   length=1.5
   radius=1.75
   shapeName=dyn_headFront_02_l_capsule
   center=<2.0846 0.0002 -0.2062>
   rotation=<rot -90 0.0198 -0.0199>
  }]
 dyn_headFront_02_r [{
   shapeType=Capsule
   length=1.25
   radius=1.75
   shapeName=dyn_headFront_02_r_capsule
   center=<2.4078 -0.2126 0.2512>
   rotation=<rot 79.9935 -179.9999 -179.9999>
  }]
 dyn_headFront_03_l [{shapeType=Sphere, radius=2.2652, shapeName=dyn_headFront_03_l_sphere, center=<2.1287 0.1478 -0.2765>}]
 dyn_headFront_03_r [{
   shapeType=Capsule
   length=1.75
   radius=1.25
   shapeName=dyn_headFront_03_r_capsule
   center=<1.2822 0.3381 -0.226>
   rotation=<rot 80.0002 -0.0003 -0.0003>
  }]
 dyn_head_01_Base_l [{shapeType=Capsule, length=0.505, radius=0.505, shapeName=dyn_head_01_Base_l_capsule, rotation=<rot 0 -90 0>}]
 dyn_head_01_Base_r [{shapeType=Capsule, length=0.505, radius=0.505, shapeName=dyn_head_01_Base_r_capsule, rotation=<rot 0 -90 0>}]
 dyn_head_02_l [{shapeType=Capsule, length=1, radius=1.75, shapeName=dyn_head_02_l_capsule, center=<1.6992 0 0>, rotation=<rot 90 0.0395 0.0394>}]
 dyn_head_02_r [{shapeType=Capsule, length=0.75, radius=2, shapeName=dyn_head_02_r_capsule, center=<2.2358 -0.8589 0.3944>, rotation=<rot 79.9872 -180 -180>}]
 dyn_head_03_l [{shapeType=Sphere, radius=3.4766, shapeName=dyn_head_03_l_sphere, center=<3.6862 -0.4571 1.3458>}]
 dyn_head_03_r [{
   shapeType=Capsule
   length=5.5387
   radius=2.5951
   shapeName=dyn_head_03_r_capsule
   center=<-0.9421 -1.6841 0.288>
   rotation=<rot -6.6774 -89.0304 -7.4661>
  }]
 head [{shapeType=Capsule, length=8.7075, radius=5.8174, center=<2.3804 1.6274 0.1518>, rotation=<rot 90 0 -7.0175>}]
 neck_01 [{shapeType=Capsule, length=10, radius=6.5487, center=<2 0 0>, rotation=<rot 90 0 0>}]
 neck_02 [{shapeType=Capsule, length=0.5225, radius=4.9374, rotation=<rot 90 0 0>}]
 spine_05 [{shapeType=Capsule, length=14, radius=7.2, center=<6.9164 -2.001 0.6251>, rotation=<rot 0 -90 0>}]
}
constraints [
 {
  bone1Name=head
  bone2Name=neck_02
  swing1LimitDegrees=45
  swing1Motion=Free
  swing2LimitDegrees=45
  swing2Motion=Free
  twistLimitDegrees=45
  twistMotion=Free
 }
 {
  bone1Name=neck_02
  bone2Name=neck_01
  swing1LimitDegrees=45
  swing1Motion=Free
  swing2LimitDegrees=45
  swing2Motion=Free
  twistLimitDegrees=45
  twistMotion=Free
 }
 {
  bone1Name=dyn_headBack_Base
  bone2Name=head
  swing1LimitDegrees=45
  swing1Motion=Free
  swing2LimitDegrees=45
  swing2Motion=Free
  twistLimitDegrees=45
  twistMotion=Free
 }
 {
  bone1Name=dyn_headBack_02_c
  bone2Name=dyn_headBack_Base
  swing1LimitDegrees=10
  swing2LimitDegrees=10
  twistLimitDegrees=8
  twistMotion=Locked
 }
 {
  bone1Name=dyn_headBack_03_c
  bone2Name=dyn_headBack_02_c
  swing1LimitDegrees=5
  swing2LimitDegrees=5
  twistLimitDegrees=8
  twistMotion=Locked
 }
 {bone1Name=dyn_headBack_02_l, bone2Name=dyn_headBack_Base, swing1LimitDegrees=24, swing2LimitDegrees=12, twistLimitDegrees=8}
 {bone1Name=dyn_headBack_03_l, bone2Name=dyn_headBack_02_l, swing1LimitDegrees=18, swing2LimitDegrees=12, twistLimitDegrees=8}
 {bone1Name=dyn_headBack_02_r, bone2Name=dyn_headBack_Base, swing1LimitDegrees=24, swing2LimitDegrees=8, twistLimitDegrees=4}
 {bone1Name=dyn_headBack_03_r, bone2Name=dyn_headBack_02_r, swing1LimitDegrees=12, swing2LimitDegrees=12, twistLimitDegrees=4}
 {
  bone1Name=clavicle_r
  bone2Name=spine_05
  swing1LimitDegrees=45
  swing1Motion=Free
  swing2LimitDegrees=45
  swing2Motion=Free
  twistLimitDegrees=45
  twistMotion=Free
 }
 {
  bone1Name=neck_01
  bone2Name=spine_05
  swing1LimitDegrees=45
  swing1Motion=Free
  swing2LimitDegrees=45
  swing2Motion=Free
  twistLimitDegrees=45
  twistMotion=Free
 }
 {bone1Name=dyn_hatWing_01_r, bone2Name=head, swing1LimitDegrees=8, swing2LimitDegrees=8, twistLimitDegrees=8}
 {bone1Name=dyn_hatWing_02_r, bone2Name=dyn_hatWing_01_r, swing1LimitDegrees=15, swing2LimitDegrees=12, twistLimitDegrees=6}
 {bone1Name=dyn_hatWing_01_l, bone2Name=head, swing1LimitDegrees=8, swing2LimitDegrees=8, twistLimitDegrees=8}
 {bone1Name=dyn_hatWing_02_l, bone2Name=dyn_hatWing_01_l, swing1LimitDegrees=15, swing2LimitDegrees=12, twistLimitDegrees=6}
 {
  bone1Name=dyn_hairFrontBounce_01_Base
  bone2Name=head
  swing1LimitDegrees=45
  swing1Motion=Free
  swing2LimitDegrees=45
  swing2Motion=Free
  twistLimitDegrees=45
  twistMotion=Free
 }
 {bone1Name=dyn_hairFrontBounce_01, bone2Name=dyn_hairFrontBounce_01_Base, swing1LimitDegrees=20, swing2LimitDegrees=20, twistLimitDegrees=10}
 {bone1Name=dyn_hairFrontBounce_02, bone2Name=dyn_hairFrontBounce_01, swing1LimitDegrees=20, swing2LimitDegrees=20, twistLimitDegrees=10}
 {
  bone1Name=dyn_headFront_01_Base_r
  bone2Name=head
  swing1LimitDegrees=45
  swing1Motion=Free
  swing2LimitDegrees=45
  swing2Motion=Free
  twistLimitDegrees=45
  twistMotion=Free
 }
 {bone1Name=dyn_headFront_01_r, bone2Name=dyn_headFront_01_Base_r, swing1LimitDegrees=8, swing2LimitDegrees=8, twistLimitDegrees=8}
 {bone1Name=dyn_headFront_02_r, bone2Name=dyn_headFront_01_r, swing1LimitDegrees=20, swing2LimitDegrees=12, twistLimitDegrees=10}
 {bone1Name=dyn_headFront_03_r, bone2Name=dyn_headFront_02_r, swing1LimitDegrees=20, swing2LimitDegrees=12, twistLimitDegrees=10}
 {
  bone1Name=dyn_headFront_01_Base_l
  bone2Name=head
  swing1LimitDegrees=45
  swing1Motion=Free
  swing2LimitDegrees=45
  swing2Motion=Free
  twistLimitDegrees=45
  twistMotion=Free
 }
 {bone1Name=dyn_headFront_01_l, bone2Name=dyn_headFront_01_Base_l, swing1LimitDegrees=25, swing2LimitDegrees=15, twistLimitDegrees=8}
 {bone1Name=dyn_headFront_02_l, bone2Name=dyn_headFront_01_l, swing1LimitDegrees=25, swing2LimitDegrees=15, twistLimitDegrees=8}
 {bone1Name=dyn_headFront_03_l, bone2Name=dyn_headFront_02_l, swing1LimitDegrees=25, swing2LimitDegrees=15, twistLimitDegrees=8}
 {
  bone1Name=dyn_head_01_Base_l
  bone2Name=head
  swing1LimitDegrees=45
  swing1Motion=Free
  swing2LimitDegrees=45
  swing2Motion=Free
  twistLimitDegrees=45
  twistMotion=Free
 }
 {bone1Name=dyn_head_02_l, bone2Name=dyn_head_01_Base_l, swing1LimitDegrees=12, swing2LimitDegrees=12, twistLimitDegrees=8}
 {bone1Name=dyn_head_03_l, bone2Name=dyn_head_02_l, swing1LimitDegrees=12, swing2LimitDegrees=12, twistLimitDegrees=8}
 {
  bone1Name=dyn_head_01_Base_r
  bone2Name=head
  swing1LimitDegrees=45
  swing1Motion=Free
  swing2LimitDegrees=45
  swing2Motion=Free
  twistLimitDegrees=45
  twistMotion=Free
 }
 {bone1Name=dyn_head_02_r, bone2Name=dyn_head_01_Base_r, swing1LimitDegrees=12, swing2LimitDegrees=12, twistLimitDegrees=8}
 {bone1Name=dyn_head_03_r, bone2Name=dyn_head_02_r, swing1LimitDegrees=12, swing2LimitDegrees=12, twistLimitDegrees=8}
 {
  bone1Name=dyn_headBack_03_r
  bone2Name=dyn_headBack_03_c
  swing1LimitDegrees=45
  swing1Motion=Free
  swing2LimitDegrees=45
  swing2Motion=Free
  twistLimitDegrees=45
  twistMotion=Free
 }
 {
  bone1Name=dyn_head_03_r
  bone2Name=dyn_headBack_03_r
  swing1LimitDegrees=45
  swing1Motion=Free
  swing2LimitDegrees=45
  swing2Motion=Free
  twistLimitDegrees=45
  twistMotion=Free
 }
 {
  bone1Name=dyn_headBack_03_l
  bone2Name=dyn_headBack_03_c
  swing1LimitDegrees=45
  swing1Motion=Free
  swing2LimitDegrees=45
  swing2Motion=Free
  twistLimitDegrees=45
  twistMotion=Free
 }
]
