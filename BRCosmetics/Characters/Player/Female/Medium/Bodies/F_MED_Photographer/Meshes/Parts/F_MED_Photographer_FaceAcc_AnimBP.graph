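# Anim Blueprint for the female Photographer's face accessory part, simulating its dynamic head bones and toggling hat visibility.
# The parent pose passes through Control Rig, right-hair correctives driven by arm and head twist, two head-space RigidBody solvers, then fixed additive offsets.
# Update switches gravity for skydiving and sets frontend/backbling alphas that the AnimGraph never reads.
# Part notifies hide the hat via material parameter, restoring it once any open parachute closes.
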AnimBlueprint /BRCosmetics/Characters/Player/Female/Medium/Bodies/F_MED_Photographer/Meshes/Parts/F_MED_Photographer_FaceAcc_AnimBP
target_skeleton=&/F_MED_Photographer_FaceAcc_Skeleton
graphs [
 AnimGraph {nodes [
   RigidBody!(ComponentPose, bAlphaBoolEnabled:bool=Get photoTraverse() -> (photoTraverse:bool), ExternalForce) -> (Pose) {
    alphaInputType=Bool
    bClampLinearTranslationLimitToRefPose=true
    bOverrideWorldGravity=true
    overridePhysicsAsset=&/F_MED_Photographer_FaceAcc_Physics
    simulationSpace=BaseBoneSpace
    alphaBoolBlend {blendOption=Linear}
    baseBoneRef {boneName=spine_03}
    componentAppliedLinearAccClamp=<3000 3000 3000>
    componentLinearAccScale=<0.385 0.2 0.2>
    componentLinearVelScale=<0.2 0.1 0.1>
    overrideWorldGravity {z=-980}
   }
   n1 = Control Rig(Source=Copy Pose From Mesh() {bUseAttachedParent=true}) {
    alphaBoolBlend {blendOption=Linear}
    controlRigAssetReference {blueprintRigClass=&/F_MED_Photographer_FaceAcc_CtrlRig.F_MED_Photographer_FaceAcc_CtrlRig_C}
   }
   n2 = Get Hair_r_correction() -> (Hair_r_correction:double)
   n3 = "Transform (Modify) Bone - Bone: dyn_head_01_Base_r"(ComponentPose=Local To Component(LocalPose=@n1), Alpha=@n2, Translation="4,0,0", Rotation="30,-20,0", Scale="1,1,1") {
    rotationMode=BMM_Additive
    rotationSpace=BCS_BoneSpace
    translationMode=BMM_Additive
    translationSpace=BCS_BoneSpace
    alphaBoolBlend {blendOption=Linear}
    boneToModify {boneName=dyn_head_01_Base_r}
    rotation {pitch=30, yaw=-20}
    translation {x=4}
   }
   n4 = "Transform (Modify) Bone - Bone: dyn_head_02_r"(ComponentPose=@n3, Alpha=@n2, Translation="0,0,0", Rotation="0,-20,0", Scale="1,1,1") {
    rotationMode=BMM_Additive
    rotationSpace=BCS_BoneSpace
    alphaBoolBlend {blendOption=Linear}
    boneToModify {boneName=dyn_head_02_r}
    rotation {yaw=-20}
   }
   n5 = RigidBody!(ComponentPose=@n4, Alpha:float, OverrideWorldGravity:Vector=Get Override World Gravity() -> (Override World Gravity:Vector), ExternalForce) -> (Pose) {
    bClampLinearTranslationLimitToRefPose=true
    bOverrideWorldGravity=true
    overridePhysicsAsset=&F_MED_Photographer_FaceAcc_Physics
    simulationSpace=BaseBoneSpace
    alphaBoolBlend {blendOption=Linear}
    baseBoneRef {boneName=head}
    componentAppliedLinearAccClamp=<2000 2000 2000>
    componentLinearAccScale=<0.385 0.2 0.2>
    componentLinearVelScale=<0.2 0.1 0.1>
    overrideWorldGravity {z=-600}
   }
   n6 = RigidBody(ComponentPose=@n5, OverrideWorldGravity="0,0,-980") {
    bClampLinearTranslationLimitToRefPose=true
    bOverrideWorldGravity=true
    overridePhysicsAsset=&/F_MED_Photographer_FaceAcc_Physics_FE
    simulationSpace=BaseBoneSpace
    alphaBoolBlend {blendOption=Linear}
    baseBoneRef {boneName=head}
    componentAppliedLinearAccClamp=<2000 2000 2000>
    componentLinearAccScale=<0.385 0.2 0.2>
    componentLinearVelScale=<0.2 0.1 0.1>
    overrideWorldGravity {z=-600}
   }
   n7 = "Transform (Modify) Bone - Bone: dyn_headBack_02_c"(ComponentPose=@n6, Alpha=1, Translation="-0.000,-1.000,2", Rotation="0.000,0.000,-20.000", Scale="1,1,1") {
    rotationMode=BMM_Additive
    translationMode=BMM_Additive
    alphaBoolBlend {blendOption=Linear}
    boneToModify {boneName=dyn_headBack_02_c}
    rotation {roll=-20}
    translation {y=-1, z=2}
   }
   n8 = "Transform (Modify) Bone - Bone: dyn_headBack_03_c"(ComponentPose=@n7, Alpha=1, Translation="-0.000,0,2", Rotation="0.000,0.000,-10.000", Scale="1,1,1") {
    rotationMode=BMM_Additive
    translationMode=BMM_Additive
    alphaBoolBlend {blendOption=Linear}
    boneToModify {boneName=dyn_headBack_03_c}
    rotation {roll=-10}
    translation {z=2}
   }
   n9 = "Transform (Modify) Bone - Bone: dyn_headBack_02_l"(ComponentPose=@n8, Alpha=1, Translation="-0.000,0,1", Rotation="0.000,0.000,-10.000", Scale="1,1,1") {
    rotationMode=BMM_Additive
    translationMode=BMM_Additive
    alphaBoolBlend {blendOption=Linear}
    boneToModify {boneName=dyn_headBack_02_l}
    rotation {roll=-10}
    translation {z=1}
   }
   n10 = "Transform (Modify) Bone - Bone: dyn_headBack_03_l"(ComponentPose=@n9, Alpha=1, Translation="-1,0.000,2", Rotation="0,0,0", Scale="1,1,1") {
    rotationMode=BMM_Additive
    translationMode=BMM_Additive
    alphaBoolBlend {blendOption=Linear}
    boneToModify {boneName=dyn_headBack_03_l}
    translation {x=-1, z=2}
   }
   n11 = "Transform (Modify) Bone - Bone: dyn_headBack_02_r"(ComponentPose=@n10, Alpha=1, Translation="-1.000,-2.000,1", Rotation="0.000,0.000,-20.000", Scale="1,1,1") {
    rotationMode=BMM_Additive
    translationMode=BMM_Additive
    alphaBoolBlend {blendOption=Linear}
    boneToModify {boneName=dyn_headBack_02_r}
    rotation {roll=-20}
    translation=<-1 -2 1>
   }
   n12 = "Transform (Modify) Bone - Bone: dyn_headBack_03_r"(ComponentPose=@n11, Alpha=1, Translation="0.000,0.000,1", Rotation="-9.847,1.754,-10.151", Scale="1,1,1") {
    rotationMode=BMM_Additive
    translationMode=BMM_Additive
    alphaBoolBlend {blendOption=Linear}
    boneToModify {boneName=dyn_headBack_03_r}
    rotation=<rot -9.847 1.754 -10.151>
    translation {z=1}
   }
   n13 = "Transform (Modify) Bone - Bone: dyn_head_03_l"(ComponentPose=@n12, Alpha=1, Translation="-1,0,1", Rotation="0,0,0", Scale="1,1,1") {
    rotationMode=BMM_Additive
    translationMode=BMM_Additive
    alphaBoolBlend {blendOption=Linear}
    boneToModify {boneName=dyn_head_03_l}
    translation {x=-1, z=1}
   }
   n14 = "Transform (Modify) Bone - Bone: dyn_head_03_r"(ComponentPose=@n13, Alpha=1, Translation="3,0,1", Rotation="-0.000,30.000,-10.000", Scale="1,1,1") {
    rotationMode=BMM_Additive
    translationMode=BMM_Additive
    alphaBoolBlend {blendOption=Linear}
    boneToModify {boneName=dyn_head_03_r}
    rotation {roll=-10, yaw=30}
    translation {x=3, z=1}
   }
   Output Pose(Result=Component To Local(ComponentPose=@n14)) {layerGroup=None}
   Get photoTraverseINV() -> (photoTraverseINV:bool)
  ]}
 EventGraph {nodes [
   Event BlueprintUpdateAnimation()
   n1 = GetOwnerSkeletalMeshComponent()
   > Set Hair_r_correction(Hair_r_correction:double=float + float(
     A=GetAlphaFromReferenceTransformDelta(
      SkeletalMeshComponent=@n1
      BoneName=upperarm_r
      BaseName=spine_05
      DeltaComponentType=QuaternionTwist
      InputRange=(X=12,Y=117)
      OutputRange=(X=0,Y=0.5)
      TwistAxis="0,0,1"
     )
     B=GetAlphaFromReferenceTransformDelta(
      SkeletalMeshComponent=@n1
      BoneName=head
      BaseName=spine_05
      DeltaComponentType=QuaternionTwist
      InputRange=(X=26,Y=-47)
      OutputRange=(X=0,Y=0.5)
      TwistAxis="0,0,1"
     )
    )) -> (Output_Get:double)
   > Set Override World Gravity(Override World Gravity:Vector=SelectVector(A="0, 0,-600", B="0, 0,-980", bPickA=Get bIsSkyDiving())) -> (Output_Get:Vector)
   n2 = IsCosmeticObjectDisplayedInFrontendContext!(CosmeticObject:Object Reference) -> (ReturnValue:bool)
   > n3 = Branch(Condition=AND Boolean(A=CheckForBackbling(), B=@n2))
   > Set normal alpha(normal alpha:double=0) -> (Output_Get:double)
   > Set backpack frontend alpha(backpack frontend alpha:double=1) -> (Output_Get:double)
   > Set frontend phat alpha(frontend phat alpha:double=0.2) -> (Output_Get:double)
   n4 = Branch(execute=@n3.else, Condition=@n2)
   > Set backpack frontend alpha(backpack frontend alpha=1)
   > Set normal alpha(normal alpha=0.3)
   > Set frontend phat alpha(frontend phat alpha=0.7)
   Set backpack frontend alpha(execute=@n4.else, backpack frontend alpha=0)
   > Set normal alpha(normal alpha=1)
   > Set frontend phat alpha(frontend phat alpha=0)
   n5 = Event OnNotifyPartsStarted!() -> (OutputDelegate, then, Context:Name=None) {is=title K2Node_Event}
   > n6 = Switch on Name!(execute:Exec, Selection:Name=@n5.Context) -> (Default:Exec, removeHatPhotographer:Exec) {is=title K2Node_SwitchName, pinNames [removeHatPhotographer]}
   n7 = GetPartSkeletalMeshComponent()
   SetScalarParameterValueOnMaterials!(execute=@n6.removeHatPhotographer, self:Mesh Component Object Reference=@n7, ParameterName:Name=Hide Element 05, ParameterValue:float=1) -> (then)
   n8 = Event OnNotifyPartsStopped!() -> (OutputDelegate, then, Context:Name=None) {is=title K2Node_Event}
   > n9 = Switch on Name(Selection=@n8.Context) {pinNames [removeHatPhotographer]}
   n10 = Branch(execute=@n9.removeHatPhotographer, Condition=Get bIsParachuteOpened())
   SetScalarParameterValueOnMaterials(execute=[@n11.then, @n10.else], self=@n7, ParameterName=Hide Element 05, ParameterValue=0)
   n11 = Event OnParachuteClosed()
   n12 = GetAnimInstance(self=GetOwnerSkeletalMeshComponent())
   n13 = Is Valid(InputObject=@n12)
   n14 = Set photoTraverse(execute=@n13.Is Not Valid, photoTraverse:bool=false) -> (Output_Get:bool)
   Set photoTraverseINV(execute=[@n14.then, @n15.then], photoTraverseINV:bool=NOT Boolean(A=Get photoTraverse())) -> (Output_Get:bool)
   n15 = Set photoTraverse(
    execute=@n13.Is Valid
    photoTraverse=Montage_IsActive!(
     self:Anim Instance Object Reference=@n12
     Montage:Anim Montage Object Reference=/BRCosmetics/Animation/Game/MainPlayer/Emotes/PhotographerTraversal/Emote_Photographer_Traversal_CMF_M
    ) -> (ReturnValue:bool)
   )
   GetOwnerSkeletalMeshComponent()
  ]}
]
variables [
 backpack frontend alpha {type=number}
 frontend phat alpha {type=number}
 Hair_r_correction {type=number}
 normal alpha {type=number}
 Override World Gravity {type=Vector, default=<0 0 -980>}
 photoTraverse {type=boolean}
 photoTraverseINV {type=boolean}
]
